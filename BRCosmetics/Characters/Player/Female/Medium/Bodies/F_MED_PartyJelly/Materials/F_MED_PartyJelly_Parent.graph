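# Parent material for the Party Jelly female character body, layering the shared Slurp effect onto the 2023 character material.
# MF_Slurp processes the PreFX result, then its emissive gets a negated-Saturation desaturation, optional addition of base emissive, and a time-of-day tint that turns bluish at night.
# The result blends over the base, optionally limited by the Tech Art Mask red channel, then feeds the PostFX function.
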Material /BRCosmetics/Characters/Player/Female/Medium/Bodies/F_MED_PartyJelly/Materials/F_MED_PartyJelly_Parent
graph {
 properties {
  bUseMaterialAttributes=true
  bUsedWithMeshDeformer=true
  bUsedWithMorphTargets=true
  bUsedWithSkeletalMesh=true
  pixelDepthOffsetMode=PDOM_Legacy
  shadingModel=MSM_FromMaterialExpression
  usedShadingModels=Default Lit | Subsurface | Clear Coat | Subsurface Profile | Cloth
 }
 nodes [
  n1 = MaterialFunctionCall() {materialFunction=&MF_CharacterParent_2023_PreFX}
  n2 = StaticSwitchParameter(True=ComponentMask(None=@n1.Tech Art Mask) {r=true}, False=Constant() {r=1}) {parameterName=Use TechArt Mask for Slurp}
  n3 = MaterialFunctionCall!(Alpha=@n2, AO, BubblesMask, SlurpNormalDistortionMask, Pre Skinned Position, Material=@n1.Result) {materialFunction=/Game/Weapons/WeaponSkins/MaterialFunction/MF_Slurp}
  n4 = GetMaterialAttributes(None=@n1.Result) {attributeGetTypes [#1]}
  n5 = Desaturation(
   None=GetMaterialAttributes(None=@n3.Result) {attributeGetTypes [#1]}.Emissive Color
   Fraction=Multiply(A=ScalarParameter() {parameterName=Saturation}) {constB=-1}
  )
  n6 = Constant3Vector() {constant=<rgba 1 1 1 0>}
  n7 = Multiply(
   A=StaticSwitchParameter(True=Add(A=@n4.Emissive Color, B=@n5), False=@n5) {parameterName=Add Base Emissive}
   B=MaterialFunctionCall(Color_Morning=@n6, Color_Day=@n6, Color_Evening=@n6, Color_Night=Constant3Vector() {constant=<rgba 0.5 0.5833 1 0>}) {materialFunction=&Color_Time_of_Day}.Result
  )
  MaterialFunctionCall(
   Material=BlendMaterialAttributes(
    A=@n1.Result
    B=SetMaterialAttributes(MaterialAttributes=@n3.Result, Emissive Color=@n7) {attributeSetTypes [#1]}
    Alpha=@n2
   )
   M=@n1.M
   Pre Skinned Local Position=@n1.M
   Distance Blend=@n1.M
   Fresnel=@n1.M
  ) {materialFunction=&MF_CharacterParent_2023_PostFX}
 ]
}
