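# Physics asset for the F_MED_PartyJelly female player body cosmetic, defining capsule bodies for the shirt's dynamic bones, upper chest and upper arms.
# Seven front, side and back shirt bones each constrain to spine_04 with twist locked and narrow swing limits, mostly on one axis.
# Extra constraints with free angular motion pair neighboring shirt bones and tie both upper-arm twist bodies to spine_04.
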PhysicsAsset /BRCosmetics/Characters/Player/Female/Medium/Bodies/F_MED_PartyJelly/Meshes/F_MED_PartyJelly_Physics
bodies {
 dyn_back_shirt_01_l [{shapeType=Capsule, length=11, radius=3, shapeName=dyn_back_shirt_01_l_capsule, center=<3.7399 0 1.3>, rotation=<rot 1.2924 -4.8304 -80.0546>}]
 dyn_back_shirt_01_r [{shapeType=Capsule, length=11, radius=3, shapeName=dyn_back_shirt_01_r_capsule, center=<3.7489 0 1.3>, rotation=<rot 1.2923 4.8304 80.0546>}]
 dyn_front_shirt_01_c [{shapeType=Capsule, length=6, radius=3, shapeName=dyn_front_shirt_01_c_capsule, center=<2.9107 0 1.5>, rotation=<rot 0 0 90>}]
 dyn_front_shirt_01_l [{shapeType=Capsule, length=8, radius=3, shapeName=dyn_front_shirt_01_l_capsule, center=<2.9107 0 1.5>, rotation=<rot 0.0002 0.0005 89.9999>}]
 dyn_front_shirt_01_r [{
   shapeType=Capsule
   length=12.5
   radius=2.5
   shapeName=dyn_front_shirt_01_r_capsule
   center=<2.9107 0.1248 1.5>
   rotation=<rot 0 -0.0005 80.0001>
  }]
 dyn_shirt_01_l [
  {shapeType=Capsule, length=16, radius=2, shapeName=dyn_shirt_01_l_capsule, center=<4.598 0 -1.4>, rotation=<rot -1.728 -9.8492 80.1486>}
  {
   shapeType=Capsule
   length=21.5
   radius=3
   shapeName=dyn_shirt_01_l_capsule
   center=<-0.6682 0 -1.3936>
   rotation=<rot -2.576 -9.6648 -99.7824>
  }
 ]
 dyn_shirt_01_r [{
   shapeType=Capsule
   length=15
   radius=3
   shapeName=dyn_shirt_01_r_capsule
   center=<-0.2792 -0.2424 -1.3749>
   rotation=<rot -1.728 9.8485 -80.1484>
  }]
 spine_04 [
  {
   shapeType=Capsule
   length=1.5
   radius=9
   shapeName=spine_04_capsule
   center=<-6.0081 2.0065 -0.0002>
   rotation=<rot 0.0006 -175.0016 89.9986>
  }
  {shapeType=Capsule, length=8, radius=7, shapeName=spine_04_capsule, center=<-5.8633 2.1792 0>}
  {shapeType=Capsule, length=11, radius=9.5, shapeName=spine_04_capsule, center=<-6.7508 1.7952 0>, rotation=<rot -90 0 -0.0002>}
 ]
 upperarm_twist_01_l [{
   shapeType=Capsule
   length=13.3845
   radius=5.5133
   shapeName=upperarm_twist_01_l_capsule
   center=<8.0739 0.1822 0.5202>
   rotation=<rot -0.0001 -94.7621 -77.0882>
  }]
 upperarm_twist_01_r [{
   shapeType=Capsule
   length=33.3845
   radius=5.0133
   shapeName=upperarm_twist_01_r_capsule
   center=<-9.6881 -0.0477 -5.2888>
   rotation=<rot -0.0003 -94.762 97.9118>
  }]
}
constraints [
 {bone1Name=dyn_shirt_01_r, bone2Name=spine_04, swing1LimitDegrees=0, swing2LimitDegrees=20, twistLimitDegrees=0}
 {bone1Name=dyn_shirt_01_l, bone2Name=spine_04, swing1LimitDegrees=0, swing2LimitDegrees=20, twistLimitDegrees=0}
 {bone1Name=dyn_front_shirt_01_r, bone2Name=spine_04, swing1LimitDegrees=5, swing2LimitDegrees=25, twistLimitDegrees=0}
 {bone1Name=dyn_front_shirt_01_l, bone2Name=spine_04, swing1LimitDegrees=5, swing2LimitDegrees=25, twistLimitDegrees=0}
 {bone1Name=dyn_front_shirt_01_c, bone2Name=spine_04, swing1LimitDegrees=5, swing2LimitDegrees=25, twistLimitDegrees=0}
 {bone1Name=dyn_back_shirt_01_r, bone2Name=spine_04, swing1LimitDegrees=2, swing2LimitDegrees=30, twistLimitDegrees=0}
 {bone1Name=dyn_back_shirt_01_l, bone2Name=spine_04, swing1LimitDegrees=2, swing2LimitDegrees=30, twistLimitDegrees=0}
 {
  bone1Name=dyn_back_shirt_01_l
  bone2Name=dyn_back_shirt_01_r
  swing1LimitDegrees=45
  swing1Motion=Free
  swing2LimitDegrees=45
  swing2Motion=Free
  twistLimitDegrees=45
  twistMotion=Free
 }
 {
  bone1Name=upperarm_twist_01_l
  bone2Name=spine_04
  swing1LimitDegrees=45
  swing1Motion=Free
  swing2LimitDegrees=45
  swing2Motion=Free
  twistLimitDegrees=45
  twistMotion=Free
 }
 {
  bone1Name=upperarm_twist_01_r
  bone2Name=spine_04
  swing1LimitDegrees=45
  swing1Motion=Free
  swing2LimitDegrees=45
  swing2Motion=Free
  twistLimitDegrees=45
  twistMotion=Free
 }
 {
  bone1Name=dyn_shirt_01_r
  bone2Name=dyn_front_shirt_01_r
  swing1LimitDegrees=45
  swing1Motion=Free
  swing2LimitDegrees=45
  swing2Motion=Free
  twistLimitDegrees=45
  twistMotion=Free
 }
 {
  bone1Name=dyn_shirt_01_r
  bone2Name=dyn_back_shirt_01_r
  swing1LimitDegrees=45
  swing1Motion=Free
  swing2LimitDegrees=45
  swing2Motion=Free
  twistLimitDegrees=45
  twistMotion=Free
 }
 {
  bone1Name=dyn_shirt_01_r
  bone2Name=dyn_front_shirt_01_r
  swing1LimitDegrees=45
  swing1Motion=Free
  swing2LimitDegrees=45
  swing2Motion=Free
  twistLimitDegrees=45
  twistMotion=Free
 }
 {
  bone1Name=dyn_shirt_01_l
  bone2Name=dyn_front_shirt_01_l
  swing1LimitDegrees=45
  swing1Motion=Free
  swing2LimitDegrees=45
  swing2Motion=Free
  twistLimitDegrees=45
  twistMotion=Free
 }
 {
  bone1Name=dyn_shirt_01_l
  bone2Name=dyn_back_shirt_01_l
  swing1LimitDegrees=45
  swing1Motion=Free
  swing2LimitDegrees=45
  swing2Motion=Free
  twistLimitDegrees=45
  twistMotion=Free
 }
]
