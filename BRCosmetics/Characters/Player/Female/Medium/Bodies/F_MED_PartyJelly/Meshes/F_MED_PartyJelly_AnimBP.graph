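# Post-process style AnimBP for the F_MED_PartyJelly character body, adding secondary motion to a pose copied from the parent mesh.
# A right bracelet constraint feeds RigidBody, AnimDynamics springs for bracelets and shirt sleeves, then a Control Rig.
# Each update computes gravity override from a parameter asset for RigidBody and bracelets.
# Weapon changes set PickAxe_Enable false for melee, disabling bracelet springs; the Control Rig receives the inverse flag.
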AnimBlueprint /BRCosmetics/Characters/Player/Female/Medium/Bodies/F_MED_PartyJelly/Meshes/F_MED_PartyJelly_AnimBP
target_skeleton=&/F_MED_PartyJelly_Skeleton
graphs [
 AnimGraph {nodes [
   n1 = "Constraint - base_bracelet_01_r (PP)"!(
    ComponentPose=Local To Component(LocalPose=Copy Pose From Mesh() {bUseAttachedParent=true})
    Alpha=1
    ConstraintWeights_0=0.4
    ConstraintWeights_1:float=0.6
   ) -> (Pose) {
    alphaBoolBlend {blendOption=Linear}
    boneToModify {boneName=base_bracelet_01_r}
    constraintSetup [{targetBone {boneName=hand_r}}, {targetBone {boneName=lowerarm_r}}]
    constraintWeights [1,1]
   }
   n2 = RigidBody!(ComponentPose=@n1, Alpha, OverrideWorldGravity:Vector=Get GravityOverride(), ExternalForce) -> (Pose) {
    bOverrideWorldGravity=true
    overridePhysicsAsset=&/F_MED_PartyJelly_Physics
    simulationSpace=BaseBoneSpace
    alphaBoolBlend {blendOption=Linear}
    baseBoneRef {boneName=pelvis}
    overrideWorldGravity {z=-980}
    simSpaceSettings {maxAngularAcceleration=80, maxAngularVelocity=20, maxLinearAcceleration=3500, maxLinearVelocity=700, velocityScaleZ=0.4, worldAlpha=1}
   }
   n3 = Get GravityOverride()
   n4 = Get PickAxe_Enable() -> (PickAxe_Enable:bool)
   n5 = Anim Dynamics - Bone: base_bracelet_01_l!(ComponentPose=@n2, bAlphaBoolEnabled:bool=@n4, GravityOverride:Vector=@n3, ExternalForce="0,0,0") -> (Pose) {
    alphaInputType=Bool
    angularDampingOverride=1
    angularSpringConstant=10000
    bAngularSpring=true
    bOverrideAngularDamping=true
    bUseGravityOverride=true
    gravityScale=0.35
    simulationSpace=BoneRelative
    alphaBoolBlend {blendOption=Cubic}
    boundBone {boneName=base_bracelet_01_l}
    componentAppliedLinearAccClamp=<0 0 0>
    gravityOverride {z=-980}
    physicsBodyDefinitions [{
      sphereCollisionRadius=0
      boundBone {boneName=base_bracelet_01_l}
      boxExtents=<25 25 25>
      constraintSetup {angularLimitsMax=<0 0.2 -0.5>, angularLimitsMin=<0 -0.2 -1>, angularTarget=<1 0 0>}
      localJointOffset=<0 0 0>
     }]
    relativeSpaceBone {boneName=root}
   }
   n6 = Anim Dynamics - Bone: base_bracelet_uppr_01_l!(ComponentPose=@n5, bAlphaBoolEnabled:bool=@n4, GravityOverride:Vector=@n3, ExternalForce="0,0,0") -> (Pose) {
    alphaInputType=Bool
    angularDampingOverride=1
    angularSpringConstant=14000
    bAngularSpring=true
    bOverrideAngularDamping=true
    bUseGravityOverride=true
    gravityScale=0.35
    simulationSpace=BoneRelative
    alphaBoolBlend {blendOption=Cubic}
    boundBone {boneName=base_bracelet_uppr_01_l}
    componentAppliedLinearAccClamp=<0 0 0>
    gravityOverride {z=-980}
    physicsBodyDefinitions [{
      sphereCollisionRadius=0
      boundBone {boneName=base_bracelet_uppr_01_l}
      boxExtents=<30 30 30>
      constraintSetup {angularLimitsMax=<0 0 3>, angularLimitsMin=<0 0 -10>, angularTarget=<1 0 0>}
      localJointOffset=<0 0 0>
     }]
    relativeSpaceBone {boneName=root}
   }
   n7 = Anim Dynamics - Bone: base_bracelet_01_r!(ComponentPose=@n6, bAlphaBoolEnabled:bool=@n4, GravityOverride:Vector=@n3, ExternalForce="0,0,0") -> (Pose) {
    alphaInputType=Bool
    angularDampingOverride=1
    angularSpringConstant=14000
    bAngularSpring=true
    bOverrideAngularDamping=true
    bUseGravityOverride=true
    gravityScale=0.35
    simulationSpace=BoneRelative
    alphaBoolBlend {blendOption=Cubic}
    boundBone {boneName=base_bracelet_01_r}
    componentAppliedLinearAccClamp=<0 0 0>
    gravityOverride {z=-980}
    physicsBodyDefinitions [{
      sphereCollisionRadius=0
      boundBone {boneName=base_bracelet_01_r}
      boxExtents=<30 30 30>
      constraintSetup {angularLimitsMax=<0 0 3>, angularLimitsMin=<0 0 -10>, angularTarget=<1 0 0>}
      localJointOffset=<10 0 0>
     }]
    relativeSpaceBone {boneName=root}
   }
   n8 = Anim Dynamics - Bone: dyn_shirt_sleeve_01_l!(ComponentPose=@n7, bAlphaBoolEnabled:bool=True, ExternalForce="0,0,0") -> (Pose) {
    alphaInputType=Bool
    angularDampingOverride=0.9999
    angularSpringConstant=4000
    bAngularSpring=true
    bOverrideAngularDamping=true
    bUseGravityOverride=true
    bUseSphericalLimits=true
    gravityScale=0.35
    simulationSpace=RootRelative
    alphaBoolBlend {blendOption=Cubic}
    boundBone {boneName=dyn_shirt_sleeve_01_l}
    componentLinearAccScale=<1 1 1>
    componentLinearVelScale=<1 1 1>
    gravityOverride {z=-980}
    physicsBodyDefinitions [{
      boundBone {boneName=dyn_shirt_sleeve_01_l}
      boxExtents=<20 30 30>
      constraintSetup {angularLimitsMax=<1 0.1 1>, angularLimitsMin=<-1 -0.1 -1>, angularTarget=<1 0 0>}
      localJointOffset=<6 0 0>
     }]
    planarLimits [{drivingBone {boneName=spine_05}, planeTransform {location=<-13.047 0.0002 -12.9746>, rotation=<rot -15 179.9991 -180>}}]
    sphericalLimits [{limitRadius=8, drivingBone {boneName=upperarm_twist_02_l}, sphereLocalOffset=<0 0 0>}]
   }
   n9 = Anim Dynamics - Bone: dyn_shirt_sleeve_01_r!(ComponentPose=@n8, bAlphaBoolEnabled:bool=True, ExternalForce="0,0,0") -> (Pose) {
    alphaInputType=Bool
    angularDampingOverride=0.9999
    angularSpringConstant=4500
    bAngularSpring=true
    bOverrideAngularDamping=true
    bUseGravityOverride=true
    bUseSphericalLimits=true
    gravityScale=0.35
    simulationSpace=RootRelative
    alphaBoolBlend {blendOption=Cubic}
    boundBone {boneName=dyn_shirt_sleeve_01_r}
    componentLinearAccScale=<1 1 1>
    componentLinearVelScale=<1 1 1>
    gravityOverride {z=-980}
    physicsBodyDefinitions [{
      boundBone {boneName=dyn_shirt_sleeve_01_r}
      boxExtents=<20 30 30>
      constraintSetup {angularLimitsMax=<3 0.3 3>, angularLimitsMin=<-3 -0.3 -3>, angularTarget=<1 0 0>}
      localJointOffset=<6 0 0>
     }]
    planarLimits [{drivingBone {boneName=spine_05}, planeTransform {location=<-9.9939 0 8.0353>, rotation=<rot -15 -0.0002 0>}}]
    sphericalLimits [{limitRadius=8, drivingBone {boneName=upperarm_twist_02_r}, sphereLocalOffset=<0 0 0>}]
   }
   n10 = Anim Dynamics - Bone: dyn_down_shirt_sleeve_01_l!(ComponentPose=@n9, bAlphaBoolEnabled:bool=True, ExternalForce="0,0,0") -> (Pose) {
    alphaInputType=Bool
    angularDampingOverride=1
    angularSpringConstant=7000
    bAngularSpring=true
    bOverrideAngularDamping=true
    bUseGravityOverride=true
    bUseSphericalLimits=true
    gravityScale=0.35
    simulationSpace=RootRelative
    alphaBoolBlend {blendOption=Cubic}
    boundBone {boneName=dyn_down_shirt_sleeve_01_l}
    componentLinearAccScale=<1 1 1>
    componentLinearVelScale=<1 1 1>
    gravityOverride {z=-980}
    physicsBodyDefinitions [{
      boundBone {boneName=dyn_down_shirt_sleeve_01_l}
      boxExtents=<20 30 30>
      constraintSetup {angularLimitsMax=<5 5 5>, angularLimitsMin=<-5 -5 -5>, angularTarget=<1 0 0>}
      localJointOffset=<4 0 0>
     }]
    sphericalLimits [{limitRadius=8, drivingBone {boneName=upperarm_twist_02_l}, sphereLocalOffset=<0 0 0>}]
   }
   n11 = Anim Dynamics - Bone: dyn_uppr_shirt_sleeve_01_l!(ComponentPose=@n10, bAlphaBoolEnabled:bool=True, ExternalForce="0,0,0") -> (Pose) {
    alphaInputType=Bool
    angularDampingOverride=1
    angularSpringConstant=7000
    bAngularSpring=true
    bOverrideAngularDamping=true
    bUseGravityOverride=true
    bUseSphericalLimits=true
    gravityScale=0.35
    simulationSpace=RootRelative
    alphaBoolBlend {blendOption=Cubic}
    boundBone {boneName=dyn_uppr_shirt_sleeve_01_l}
    componentLinearAccScale=<1 1 1>
    componentLinearVelScale=<1 1 1>
    gravityOverride {z=-980}
    physicsBodyDefinitions [{
      boundBone {boneName=dyn_uppr_shirt_sleeve_01_l}
      boxExtents=<20 30 30>
      constraintSetup {angularLimitsMax=<5 5 5>, angularLimitsMin=<-5 -5 -5>, angularTarget=<1 0 0>}
      localJointOffset=<4 0 0>
     }]
    sphericalLimits [{limitRadius=8, drivingBone {boneName=upperarm_twist_02_l}, sphereLocalOffset=<0 0 0>}]
   }
   n12 = Anim Dynamics - Bone: base_down_shirt_sleeve_01_r!(ComponentPose=@n11, bAlphaBoolEnabled:bool=True, ExternalForce="0,0,0") -> (Pose) {
    alphaInputType=Bool
    angularDampingOverride=0.9999
    angularSpringConstant=8000
    bAngularSpring=true
    bOverrideAngularDamping=true
    bUseGravityOverride=true
    bUseSphericalLimits=true
    gravityScale=0.35
    simulationSpace=RootRelative
    alphaBoolBlend {blendOption=Cubic}
    boundBone {boneName=base_down_shirt_sleeve_01_r}
    componentLinearAccScale=<1 1 1>
    componentLinearVelScale=<1 1 1>
    gravityOverride {z=-980}
    physicsBodyDefinitions [{
      boundBone {boneName=base_down_shirt_sleeve_01_r}
      boxExtents=<20 30 30>
      constraintSetup {angularLimitsMax=<5 5 5>, angularLimitsMin=<-5 -5 -5>, angularTarget=<1 0 0>}
      localJointOffset=<4 0 0>
     }]
    sphericalLimits [{limitRadius=8, drivingBone {boneName=upperarm_twist_02_r}, sphereLocalOffset=<0 0 0>}]
   }
   n13 = Anim Dynamics - Bone: base_uppr_shirt_sleeve_01_r(ComponentPose=@n12, Alpha=1, ExternalForce="0,0,0") {
    angularDampingOverride=0.9999
    angularSpringConstant=8000
    bAngularSpring=true
    bOverrideAngularDamping=true
    bUseGravityOverride=true
    bUseSphericalLimits=true
    gravityScale=0.35
    simulationSpace=RootRelative
    alphaBoolBlend {blendOption=Linear}
    boundBone {boneName=base_uppr_shirt_sleeve_01_r}
    componentLinearAccScale=<1 1 1>
    componentLinearVelScale=<1 1 1>
    gravityOverride {z=-980}
    physicsBodyDefinitions [{
      boundBone {boneName=base_uppr_shirt_sleeve_01_r}
      boxExtents=<20 30 30>
      constraintSetup {angularLimitsMax=<5 5 5>, angularLimitsMin=<-5 -5 -5>, angularTarget=<1 0 0>}
      localJointOffset=<4 0 0>
     }]
    sphericalLimits [{limitRadius=8, drivingBone {boneName=upperarm_twist_02_r}, sphereLocalOffset=<0 0 0>}]
   }
   n14 = Control Rig!(Source=Component To Local(ComponentPose=@n13), Alpha, PickAxe_Enable:bool=NOT Boolean(A=Get PickAxe_Enable())) -> (Pose) {
    alphaBoolBlend {blendOption=Linear}
    controlRigAssetReference {blueprintRigClass=&/F_MED_PartyJelly_CtrlRig.F_MED_PartyJelly_CtrlRig_C}
   }
   Output Pose(Result=@n14) {layerGroup=None}
   Get PickAxe_Enable_Amin() -> (PickAxe_Enable_Amin:bool)
  ]}
 EventGraph {nodes [
   Event BlueprintUpdateAnimation()
   > n1 = "Gravity Override State Function (Parameter Asset) (Nativized)"(
    CharacterPartAnimInstance=Self-Reference()
    GameTimeInSeconds=GetGameTimeInSeconds()
    Parameters=&/M_MED_PartyJelly_Parameters
   )
   > n2 = Set GravityOverride(GravityOverride=@n1)
   > Set GravityOverride(GravityOverride=@n2)
   Event OnWeaponChanged()
   > n3 = Cast To FortPlayerAnimInstance(Object=GetAnimInstance(self=GetOwnerSkeletalMeshComponent()))
   > n4 = Switch on EFortWeaponCoreAnimation!(execute, Selection:EFortWeaponCoreAnimation Enum=Get WeaponCoreAnim(self=@n3)) -> (
    Melee:Exec
    Pistol:Exec
    Shotgun:Exec
    PaperBlueprint:Exec
    Rifle:Exec
    MeleeOneHand:Exec
    MachinePistol:Exec
    RocketLauncher:Exec
    GrenadeLauncher:Exec
    GoingCommando:Exec
    AssaultRifle:Exec
    TacticalShotgun:Exec
    SniperRifle:Exec
    TrapPlacement:Exec
    ShoulderLauncher:Exec
    AbilityDecoTool:Exec
    Crossbow:Exec
    C4:Exec
    RemoteControl:Exec
    DualWield:Exec
    AR_BullPup:Exec
    AR_ForwardGrip:Exec
    MedPackPaddles:Exec
    SMG_P90:Exec
    AR_DrumGun:Exec
    Consumable_Small:Exec
    Consumable_Large:Exec
    Balloon:Exec
    MountedTurret:Exec
    CreativeTool:Exec
    ExplosiveBow:Exec
    AshtonIndigo:Exec
    AshtonChicago:Exec
    MeleeDualWield:Exec
    Unarmed:Exec
   )
   n5 = Set PickAxe_Enable(
    execute=[
     @n4.AR_BullPup
     @n4.AR_DrumGun
     @n4.AR_ForwardGrip
     @n4.AbilityDecoTool
     @n4.AshtonChicago
     @n4.AshtonIndigo
     @n4.AssaultRifle
     @n4.Balloon
     @n4.C4
     @n4.Consumable_Large
     @n4.Consumable_Small
     @n4.CreativeTool
     @n4.Crossbow
     @n4.DualWield
     @n4.ExplosiveBow
     @n4.GoingCommando
     @n4.GrenadeLauncher
     @n4.MachinePistol
     @n4.MedPackPaddles
     @n4.MountedTurret
     @n4.PaperBlueprint
     @n4.Pistol
     @n4.RemoteControl
     @n4.Rifle
     @n4.RocketLauncher
     @n4.SMG_P90
     @n4.Shotgun
     @n4.ShoulderLauncher
     @n4.SniperRifle
     @n4.TacticalShotgun
     @n4.TrapPlacement
     @n4.Unarmed
    ]
    PickAxe_Enable:bool=true
   ) -> (Output_Get:bool)
   > PrintString!(execute, InString="To String (Boolean)"(InBool=@n5), bPrintToScreen, bPrintToLog, TextColor, Duration, Key) -> (then)
   n6 = Set PickAxe_Enable(execute=[@n4.Melee, @n4.MeleeDualWield, @n4.MeleeOneHand], PickAxe_Enable=false)
   > PrintString(InString="To String (Boolean)"(InBool=@n6))
   SelectVector(A="0,0,-3000", B=@n2, bPickA=Get isWearingBackbling())
  ]}
]
variables [
 Gravity Override Params {type=GravityOverrideParamsStruct}
 GravityOverride {type=Vector}
 isWearingBackbling {type=boolean}
 Melee {type=number}
 PickAxe_Enable {type=boolean}
 PickAxe_Enable_Amin {type=boolean}
]
