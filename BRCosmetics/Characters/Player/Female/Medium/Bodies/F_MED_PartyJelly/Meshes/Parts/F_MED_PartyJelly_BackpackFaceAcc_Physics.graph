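# Physics asset for the backpack/face accessory part of the female medium PartyJelly cosmetic body, giving its dangling dyn_hair strands physical motion.
# Seven three-bone capsule chains (front, side, back) hang from a cape sphere, with swing limits widening toward the tips and zero twist.
# Side and back chains swing on one axis, while front chains favor the other.
# Spine, clavicle and upper-arm capsules, joined by free angular constraints, provide body collision shapes.
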PhysicsAsset /BRCosmetics/Characters/Player/Female/Medium/Bodies/F_MED_PartyJelly/Meshes/Parts/F_MED_PartyJelly_BackpackFaceAcc_Physics
bodies {
 cape [{shapeType=Sphere, radius=5.625, shapeName=cape_sphere, center=<-8.3847 1.4523 0.0695>}]
 clavicle_l [{
   shapeType=Capsule
   length=22.5
   radius=7.005
   shapeName=clavicle_l_capsule
   center=<6.301 0.1459 -2.8271>
   rotation=<rot -0.0015 -84.9999 89.9999>
  }]
 clavicle_r [{
   shapeType=Capsule
   length=22.5
   radius=7.005
   shapeName=clavicle_r_capsule
   center=<-6.301 -0.1459 2.8271>
   rotation=<rot -0.0015 -84.9999 -90.0001>
  }]
 dyn_hair_01_l [{shapeType=Capsule, length=3.375, radius=1.625, shapeName=dyn_hair_01_l_capsule, rotation=<rot 90 0 0>}]
 dyn_hair_01_r [{shapeType=Capsule, length=3.375, radius=1.625, shapeName=dyn_hair_01_r_capsule, rotation=<rot 90 0 0>}]
 dyn_hair_02_l [{shapeType=Capsule, length=2.125, radius=1.875, shapeName=dyn_hair_02_l_capsule, center=<0 0 -0.1614>, rotation=<rot 90 0 0>}]
 dyn_hair_02_r [{
   shapeType=Capsule
   length=5.125
   radius=3.375
   shapeName=dyn_hair_02_r_capsule
   center=<3.2629 0.249 0.8021>
   rotation=<rot 88.2653 94.9079 85.0599>
  }]
 dyn_hair_03_l [{
   shapeType=Capsule
   length=2.25
   radius=3
   shapeName=dyn_hair_03_l_capsule
   center=<3.2095 -1.0781 0.2206>
   rotation=<rot 78.0128 -179.9993 -179.9994>
  }]
 dyn_hair_03_r [{
   shapeType=Capsule
   length=5
   radius=3.625
   shapeName=dyn_hair_03_r_capsule
   center=<2.6128 0.9355 0.8231>
   rotation=<rot 74.1033 -172.2081 172.6988>
  }]
 dyn_hair_back_01_c [{shapeType=Capsule, length=3.375, radius=1.625, shapeName=dyn_hair_back_01_c_capsule, rotation=<rot 90 0 0>}]
 dyn_hair_back_01_l [{shapeType=Capsule, length=3.375, radius=1.625, shapeName=dyn_hair_back_01_l_capsule, rotation=<rot 90 0 0>}]
 dyn_hair_back_01_r [{shapeType=Capsule, length=3.375, radius=1.625, shapeName=dyn_hair_back_01_r_capsule, rotation=<rot 90 0 0>}]
 dyn_hair_back_02_c [{shapeType=Capsule, length=2.875, radius=2.125, shapeName=dyn_hair_back_02_c_capsule, center=<0 0 0.6>, rotation=<rot 79.7253 0 0>}]
 dyn_hair_back_02_l [{shapeType=Capsule, length=3.125, radius=1.875, shapeName=dyn_hair_back_02_l_capsule, center=<0 0 0.5>, rotation=<rot 90 0 0>}]
 dyn_hair_back_02_r [{
   shapeType=Capsule
   length=2.5
   radius=2.5
   shapeName=dyn_hair_back_02_r_capsule
   center=<0.0055 0 0.2966>
   rotation=<rot 79.9997 0.0005 0.0005>
  }]
 dyn_hair_back_03_c [{
   shapeType=Capsule
   length=4.125
   radius=4
   shapeName=dyn_hair_back_03_c_capsule
   center=<1.4347 0 0.3256>
   rotation=<rot 85.3369 179.9997 179.9997>
  }]
 dyn_hair_back_03_l [{
   shapeType=Capsule
   length=2.625
   radius=3.875
   shapeName=dyn_hair_back_03_l_capsule
   center=<3.1808 0 -0.0044>
   rotation=<rot 80.1954 -0.0006 -0.0007>
  }]
 dyn_hair_back_03_r [{
   shapeType=Capsule
   length=4.875
   radius=3.625
   shapeName=dyn_hair_back_03_r_capsule
   center=<2.6357 -0.3205 -0.224>
   rotation=<rot 88.496 0.0239 0.024>
  }]
 dyn_hair_front_01_l [{shapeType=Capsule, length=2.5, radius=2, shapeName=dyn_hair_front_01_l_capsule, rotation=<rot 90 0 0>}]
 dyn_hair_front_01_r [{shapeType=Capsule, length=2.5, radius=2, shapeName=dyn_hair_front_01_r_capsule, rotation=<rot 90 0 0>}]
 dyn_hair_front_02_l [{shapeType=Capsule, length=2.75, radius=1.75, shapeName=dyn_hair_front_02_l_capsule, rotation=<rot 90 0 0>}]
 dyn_hair_front_02_r [{
   shapeType=Capsule
   length=3.75
   radius=2.5
   shapeName=dyn_hair_front_02_r_capsule
   center=<2.3152 0.0255 0.6144>
   rotation=<rot 85 -10.0011 -19.9999>
  }]
 dyn_hair_front_03_l [{shapeType=Capsule, length=5.5, radius=3.5, shapeName=dyn_hair_front_03_l_capsule, center=<1.7295 0 0.11>, rotation=<rot 90 180 178.2097>}]
 dyn_hair_front_03_r [{
   shapeType=Capsule
   length=4.625
   radius=3.25
   shapeName=dyn_hair_front_03_r_capsule
   center=<1.5992 0.1098 0.4886>
   rotation=<rot 85 -10.0026 -5>
  }]
 spine_04 [{shapeType=Capsule, length=55.38, radius=5.63, shapeName=spine_04_capsule, center=<11.2617 -0.5501 0>, rotation=<rot 0 -90 0>}]
 spine_05 [
  {shapeType=Capsule, length=39.13, radius=5.005, shapeName=spine_05_capsule, center=<3.6737 -2.3498 0>, rotation=<rot 0 -175.002 -180>}
  {shapeType=Capsule, length=5.375, radius=4.125, shapeName=spine_05_capsule, center=<10.1287 1.2555 0>}
 ]
 upperarm_l [
  {shapeType=Capsule, length=0.505, radius=0.505, shapeName=upperarm_l_capsule, rotation=<rot 0 -90 0>}
  {
   shapeType=Capsule
   length=21.38
   radius=4.38
   shapeName=upperarm_l_capsule
   center=<18.8907 -6.575 0.1161>
   rotation=<rot 0.8667 -90.0758 -90.0384>
  }
 ]
 upperarm_r [
  {shapeType=Capsule, length=0.505, radius=1, shapeName=upperarm_r_capsule, rotation=<rot 0 -90 180>}
  {
   shapeType=Capsule
   length=41.755
   radius=8.875
   shapeName=upperarm_r_capsule
   center=<-13.6606 12.7588 -2.2571>
   rotation=<rot 0.8526 -77.07 90.1575>
  }
 ]
 upperarm_twist_01_r [{
   shapeType=Capsule
   length=17.505
   radius=4.255
   shapeName=upperarm_twist_01_r_capsule
   center=<-1.221 0.6626 -0.0791>
   rotation=<rot -0.0002 -89.9998 90>
  }]
}
constraints [
 {bone1Name=dyn_hair_02_l, bone2Name=dyn_hair_01_l, swing1LimitDegrees=0, swing2LimitDegrees=40, twistLimitDegrees=0}
 {bone1Name=dyn_hair_03_l, bone2Name=dyn_hair_02_l, swing1LimitDegrees=0, swing2LimitDegrees=50, twistLimitDegrees=0}
 {bone1Name=dyn_hair_02_r, bone2Name=dyn_hair_01_r, swing1LimitDegrees=0, swing2LimitDegrees=35, twistLimitDegrees=0}
 {bone1Name=dyn_hair_03_r, bone2Name=dyn_hair_02_r, swing1LimitDegrees=0, swing2LimitDegrees=45, twistLimitDegrees=0}
 {bone1Name=dyn_hair_back_02_c, bone2Name=dyn_hair_back_01_c, swing1LimitDegrees=0, swing2LimitDegrees=40, twistLimitDegrees=0}
 {bone1Name=dyn_hair_back_03_c, bone2Name=dyn_hair_back_02_c, swing1LimitDegrees=0, swing2LimitDegrees=50, twistLimitDegrees=0}
 {bone1Name=dyn_hair_back_02_l, bone2Name=dyn_hair_back_01_l, swing1LimitDegrees=0, swing2LimitDegrees=30, twistLimitDegrees=0}
 {bone1Name=dyn_hair_back_03_l, bone2Name=dyn_hair_back_02_l, swing1LimitDegrees=0, swing2LimitDegrees=40, twistLimitDegrees=0}
 {bone1Name=dyn_hair_back_02_r, bone2Name=dyn_hair_back_01_r, swing1LimitDegrees=0, swing2LimitDegrees=25, twistLimitDegrees=0}
 {bone1Name=dyn_hair_back_03_r, bone2Name=dyn_hair_back_02_r, swing1LimitDegrees=0, swing2LimitDegrees=35, twistLimitDegrees=0}
 {bone1Name=dyn_hair_front_02_l, bone2Name=dyn_hair_front_01_l, swing1LimitDegrees=20, swing2LimitDegrees=4, twistLimitDegrees=0}
 {bone1Name=dyn_hair_front_03_l, bone2Name=dyn_hair_front_02_l, swing1LimitDegrees=30, swing2LimitDegrees=6, twistLimitDegrees=0}
 {bone1Name=dyn_hair_front_02_r, bone2Name=dyn_hair_front_01_r, swing1LimitDegrees=20, swing2LimitDegrees=4, twistLimitDegrees=0}
 {bone1Name=dyn_hair_front_03_r, bone2Name=dyn_hair_front_02_r, swing1LimitDegrees=30, swing2LimitDegrees=6, twistLimitDegrees=0}
 {bone1Name=dyn_hair_back_01_r, bone2Name=cape, swing1LimitDegrees=0, swing2LimitDegrees=15, twistLimitDegrees=0}
 {bone1Name=dyn_hair_01_r, bone2Name=cape, swing1LimitDegrees=0, swing2LimitDegrees=5, twistLimitDegrees=0}
 {bone1Name=dyn_hair_front_01_r, bone2Name=cape, swing1LimitDegrees=10, swing2LimitDegrees=2, twistLimitDegrees=0}
 {bone1Name=dyn_hair_back_01_c, bone2Name=cape, swing1LimitDegrees=0, swing2LimitDegrees=20, twistLimitDegrees=0}
 {bone1Name=dyn_hair_back_01_l, bone2Name=cape, swing1LimitDegrees=0, swing2LimitDegrees=10, twistLimitDegrees=0}
 {bone1Name=dyn_hair_01_l, bone2Name=cape, swing1LimitDegrees=0, swing2LimitDegrees=5, twistLimitDegrees=0}
 {bone1Name=dyn_hair_front_01_l, bone2Name=cape, swing1LimitDegrees=10, swing2LimitDegrees=2, twistLimitDegrees=0}
 {
  bone1Name=clavicle_l
  bone2Name=spine_05
  swing1LimitDegrees=45
  swing1Motion=Free
  swing2LimitDegrees=45
  swing2Motion=Free
  twistLimitDegrees=45
  twistMotion=Free
 }
 {
  bone1Name=clavicle_r
  bone2Name=spine_05
  swing1LimitDegrees=45
  swing1Motion=Free
  swing2LimitDegrees=45
  swing2Motion=Free
  twistLimitDegrees=45
  twistMotion=Free
 }
 {
  bone1Name=cape
  bone2Name=spine_05
  swing1LimitDegrees=45
  swing1Motion=Free
  swing2LimitDegrees=45
  swing2Motion=Free
  twistLimitDegrees=45
  twistMotion=Free
 }
 {
  bone1Name=upperarm_l
  bone2Name=clavicle_l
  swing1LimitDegrees=45
  swing1Motion=Free
  swing2LimitDegrees=45
  swing2Motion=Free
  twistLimitDegrees=45
  twistMotion=Free
 }
 {
  bone1Name=upperarm_r
  bone2Name=clavicle_r
  swing1LimitDegrees=45
  swing1Motion=Free
  swing2LimitDegrees=45
  swing2Motion=Free
  twistLimitDegrees=45
  twistMotion=Free
 }
 {
  bone1Name=upperarm_twist_01_r
  bone2Name=upperarm_r
  swing1LimitDegrees=45
  swing1Motion=Free
  swing2LimitDegrees=45
  swing2Motion=Free
  twistLimitDegrees=45
  twistMotion=Free
 }
 {
  bone1Name=spine_05
  bone2Name=spine_04
  swing1LimitDegrees=45
  swing1Motion=Free
  swing2LimitDegrees=45
  swing2Motion=Free
  twistLimitDegrees=45
  twistMotion=Free
 }
]
