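# Post-process-style secondary-motion AnimBlueprint for the PartyJelly female face accessory skeleton, posing and simulating hair and cape bones.
# The main chain layers Control Rigs and additive bone offsets driven by head rotation, movement, targeting, weapon, backpack and frontend states.
# A head-space RigidBody pass precedes backbling constraints, then spring AnimDynamics swings front hair against body-bone limits.
# Separate AnimDynamics and weighted-constraint chains also cover back and side hair.
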AnimBlueprint /BRCosmetics/Characters/Player/Female/Medium/Bodies/F_MED_PartyJelly/Meshes/Parts/F_MED_PartyJelly_FaceAcc_AnimBP
target_skeleton=&/F_MED_PartyJelly_FaceAcc_Skeleton
graphs [
 AnimGraph {nodes [
   n1 = Anim Dynamics - Bone: dyn_hair_back_01_c!(ComponentPose, bAlphaBoolEnabled:bool=True, ExternalForce="0,0,0") -> (Pose) {
    alphaInputType=Bool
    angularDampingOverride=1
    angularSpringConstant=20000
    bOverrideAngularDamping=true
    bUseGravityOverride=true
    gravityScale=0.35
    simulationSpace=RootRelative
    alphaBoolBlend {blendOption=Linear}
    boundBone {boneName=dyn_hair_back_01_c}
    componentAppliedLinearAccClamp=<0 0 0>
    physicsBodyDefinitions [{
      boundBone {boneName=dyn_hair_back_01_c}
      boxExtents=<1 1 1>
      constraintSetup {angularLimitsMax=<0 0 0>, angularLimitsMin=<0 0 0>, angularTarget=<1 0 0>}
      localJointOffset=<0 0 0>
     }]
   }
   n2 = Anim Dynamics - Bone: dyn_hair_back_02_c!(ComponentPose=@n1, bAlphaBoolEnabled:bool=True, ExternalForce="0,0,0") -> (Pose) {
    alphaInputType=Bool
    angularDampingOverride=1
    angularSpringConstant=20000
    bOverrideAngularDamping=true
    bUseGravityOverride=true
    gravityScale=0.35
    simulationSpace=RootRelative
    alphaBoolBlend {blendOption=Linear}
    boundBone {boneName=dyn_hair_back_02_c}
    componentAppliedLinearAccClamp=<0 0 0>
    physicsBodyDefinitions [{
      boundBone {boneName=dyn_hair_back_02_c}
      boxExtents=<1 1 1>
      constraintSetup {angularLimitsMax=<0 0 0>, angularLimitsMin=<0 0 0>, angularTarget=<1 0 0>}
      localJointOffset=<0 0 0>
     }]
   }
   n3 = Anim Dynamics - Bone: dyn_hair_back_03_c!(ComponentPose=@n2, bAlphaBoolEnabled:bool=True, ExternalForce="0,0,0") -> (Pose) {
    alphaInputType=Bool
    angularDampingOverride=1
    angularSpringConstant=20000
    bOverrideAngularDamping=true
    bUseGravityOverride=true
    gravityScale=0.35
    simulationSpace=RootRelative
    alphaBoolBlend {blendOption=Linear}
    boundBone {boneName=dyn_hair_back_03_c}
    componentAppliedLinearAccClamp=<0 0 0>
    physicsBodyDefinitions [{
      boundBone {boneName=dyn_hair_back_03_c}
      boxExtents=<1 1 1>
      constraintSetup {angularLimitsMax=<0 0 0>, angularLimitsMin=<0 0 0>, angularTarget=<1 0 0>}
      localJointOffset=<0 0 0>
     }]
   }
   n4 = Anim Dynamics - Bone: dyn_hair_back_01_r!(ComponentPose=@n3, bAlphaBoolEnabled:bool=True, ExternalForce="0,0,0") -> (Pose) {
    alphaInputType=Bool
    angularDampingOverride=1
    angularSpringConstant=20000
    bOverrideAngularDamping=true
    bUseGravityOverride=true
    gravityScale=0.35
    simulationSpace=RootRelative
    alphaBoolBlend {blendOption=Linear}
    boundBone {boneName=dyn_hair_back_01_r}
    componentAppliedLinearAccClamp=<0 0 0>
    physicsBodyDefinitions [{
      boundBone {boneName=dyn_hair_back_01_r}
      boxExtents=<1 1 1>
      constraintSetup {angularLimitsMax=<0 0 0>, angularLimitsMin=<0 0 0>, angularTarget=<1 0 0>}
      localJointOffset=<0 0 0>
     }]
   }
   n5 = Anim Dynamics - Bone: dyn_hair_back_02_r!(ComponentPose=@n4, bAlphaBoolEnabled:bool=True, ExternalForce="0,0,0") -> (Pose) {
    alphaInputType=Bool
    angularDampingOverride=1
    angularSpringConstant=20000
    bOverrideAngularDamping=true
    bUseGravityOverride=true
    gravityScale=0.35
    simulationSpace=RootRelative
    alphaBoolBlend {blendOption=Linear}
    boundBone {boneName=dyn_hair_back_02_r}
    componentAppliedLinearAccClamp=<0 0 0>
    physicsBodyDefinitions [{
      boundBone {boneName=dyn_hair_back_02_r}
      boxExtents=<1 1 1>
      constraintSetup {angularLimitsMax=<0 0 0>, angularLimitsMin=<0 0 0>, angularTarget=<1 0 0>}
      localJointOffset=<0 0 0>
     }]
   }
   n6 = Anim Dynamics - Bone: dyn_hair_back_03_r!(ComponentPose=@n5, bAlphaBoolEnabled:bool=True, ExternalForce="0,0,0") -> (Pose) {
    alphaInputType=Bool
    angularDampingOverride=1
    angularSpringConstant=20000
    bOverrideAngularDamping=true
    bUseGravityOverride=true
    gravityScale=0.35
    simulationSpace=RootRelative
    alphaBoolBlend {blendOption=Linear}
    boundBone {boneName=dyn_hair_back_03_r}
    componentAppliedLinearAccClamp=<0 0 0>
    physicsBodyDefinitions [{
      boundBone {boneName=dyn_hair_back_03_r}
      boxExtents=<1 1 1>
      constraintSetup {angularLimitsMax=<0 0 0>, angularLimitsMin=<0 0 0>, angularTarget=<1 0 0>}
      localJointOffset=<0 0 0>
     }]
   }
   n7 = Anim Dynamics - Bone: dyn_hair_01_r!(ComponentPose=@n6, bAlphaBoolEnabled:bool=True, ExternalForce="0,0,0") -> (Pose) {
    alphaInputType=Bool
    angularDampingOverride=1
    angularSpringConstant=20000
    bOverrideAngularDamping=true
    bUseGravityOverride=true
    gravityScale=0.35
    simulationSpace=RootRelative
    alphaBoolBlend {blendOption=Linear}
    boundBone {boneName=dyn_hair_01_r}
    componentAppliedLinearAccClamp=<0 0 0>
    physicsBodyDefinitions [{
      boundBone {boneName=dyn_hair_01_r}
      boxExtents=<1 1 1>
      constraintSetup {angularLimitsMax=<0 0 0>, angularLimitsMin=<0 0 0>, angularTarget=<1 0 0>}
      localJointOffset=<0 0 0>
     }]
   }
   n8 = Anim Dynamics - Bone: dyn_hair_02_r!(ComponentPose=@n7, bAlphaBoolEnabled:bool=True, ExternalForce="0,0,0") -> (Pose) {
    alphaInputType=Bool
    angularDampingOverride=1
    angularSpringConstant=20000
    bOverrideAngularDamping=true
    bUseGravityOverride=true
    gravityScale=0.35
    simulationSpace=RootRelative
    alphaBoolBlend {blendOption=Linear}
    boundBone {boneName=dyn_hair_02_r}
    componentAppliedLinearAccClamp=<0 0 0>
    physicsBodyDefinitions [{
      boundBone {boneName=dyn_hair_02_r}
      boxExtents=<1 1 1>
      constraintSetup {angularLimitsMax=<0 0 0>, angularLimitsMin=<0 0 0>, angularTarget=<1 0 0>}
      localJointOffset=<0 0 0>
     }]
   }
   Anim Dynamics - Bone: dyn_hair_03_r!(ComponentPose=@n8, bAlphaBoolEnabled:bool=True, ExternalForce="0,0,0") -> (Pose) {
    alphaInputType=Bool
    angularDampingOverride=1
    angularSpringConstant=20000
    bOverrideAngularDamping=true
    bUseGravityOverride=true
    gravityScale=0.35
    simulationSpace=RootRelative
    alphaBoolBlend {blendOption=Linear}
    boundBone {boneName=dyn_hair_03_r}
    componentAppliedLinearAccClamp=<0 0 0>
    physicsBodyDefinitions [{
      boundBone {boneName=dyn_hair_03_r}
      boxExtents=<1 1 1>
      constraintSetup {angularLimitsMax=<0 0 0>, angularLimitsMin=<0 0 0>, angularTarget=<1 0 0>}
      localJointOffset=<0 0 0>
     }]
   }
   n9 = "Anim Dynamics - Chain: dyn_hair_02_l -> dyn_hair_03_l"!(ComponentPose, Alpha=1, GravityOverride:Vector="0,0,0", ExternalForce="0,0,0") -> (Pose) {
    angularDampingOverride=1
    angularSpringConstant=10000
    bChain=true
    bUseGravityOverride=true
    bUsePlanarLimit=false
    bUseSphericalLimits=true
    gravityScale=0.35
    simulationSpace=RootRelative
    alphaBoolBlend {blendOption=Linear}
    boundBone {boneName=dyn_hair_02_l}
    chainEnd {boneName=dyn_hair_03_l}
    componentLinearAccScale=<1 1 1>
    componentLinearVelScale=<1 1 1>
    gravityOverride {z=-980}
    physicsBodyDefinitions [
     {
      boundBone {boneName=dyn_hair_02_l}
      boxExtents=<40 30 30>
      constraintSetup {angularLimitsMax=<0 0 0>, angularLimitsMin=<0 0 0>, angularTarget=<1 0 0>}
      localJointOffset=<0.846 0 -0.1392>
     }
     {
      boundBone {boneName=base_hair_03_l}
      boxExtents=<40 30 30>
      constraintSetup {angularLimitsMax=<0 0 0>, angularLimitsMin=<0 0 0>, angularTarget=<1 0 0>}
      localJointOffset=<0.846 0 -0.1392>
     }
     {
      boundBone {boneName=dyn_hair_03_l}
      boxExtents=<40 30 30>
      constraintSetup {angularLimitsMax=<0 0 0>, angularLimitsMin=<0 0 0>, angularTarget=<1 0 0>}
      localJointOffset=<0.846 0 -0.1392>
     }
    ]
    sphericalLimits [{limitRadius=10, drivingBone {boneName=clavicle_l}, sphereLocalOffset=<0.2739 -0.1306 0>}]
   }
   "Anim Dynamics - Chain: dyn_hair_02_r -> dyn_hair_03_r"!(ComponentPose=@n9, Alpha=1, GravityOverride:Vector="0,0,0", ExternalForce="0,0,0") -> (Pose) {
    angularDampingOverride=1
    angularSpringConstant=10000
    bChain=true
    bUseGravityOverride=true
    bUsePlanarLimit=false
    bUseSphericalLimits=true
    gravityScale=0.35
    simulationSpace=RootRelative
    alphaBoolBlend {blendOption=Linear}
    boundBone {boneName=dyn_hair_02_r}
    chainEnd {boneName=dyn_hair_03_r}
    componentLinearAccScale=<1 1 1>
    componentLinearVelScale=<1 1 1>
    gravityOverride {z=-980}
    physicsBodyDefinitions [
     {
      boundBone {boneName=dyn_hair_02_r}
      boxExtents=<40 30 30>
      constraintSetup {angularLimitsMax=<0 0 0>, angularLimitsMin=<0 0 0>, angularTarget=<1 0 0>}
      localJointOffset=<0 0 0>
     }
     {
      boundBone {boneName=base_hair_03_r}
      boxExtents=<20 30 30>
      constraintSetup {angularLimitsMax=<0 0 0>, angularLimitsMin=<0 0 0>, angularTarget=<1 0 0>}
      localJointOffset=<0 0 0>
     }
     {
      boundBone {boneName=dyn_hair_03_r}
      boxExtents=<20 30 30>
      constraintSetup {angularLimitsMax=<0 0 0>, angularLimitsMin=<0 0 0>, angularTarget=<1 0 0>}
      localJointOffset=<0.846 0 -0.1392>
     }
    ]
    sphericalLimits [{limitRadius=13, drivingBone {boneName=clavicle_r}, sphereLocalOffset=<-1.8802 -0.8075 1.0479>}]
   }
   Component To Local()
   n10 = Control Rig!(Source, Alpha, FrontEnd Not Emote:bool=Get FrontEnd_not_emote() -> (FrontEnd_not_emote:bool)) -> (Pose) {
    alphaBoolBlend {blendOption=Linear}
    controlRigAssetReference {blueprintRigClass=&/F_MED_PartyJelly_FaceAcc_CtrlRig.F_MED_PartyJelly_FaceAcc_CtrlRig_C}
   }
   n11 = "Transform (Modify) Bone - Bone: base_hair_01_r"!(
    ComponentPose=Local To Component(LocalPose=@n10)
    bAlphaBoolEnabled:bool=Get bIsMoving2D()
    Translation="0,-1,1"
    Rotation="0,-5,0"
    Scale="1,1,1"
   ) -> (Pose) {
    alphaInputType=Bool
    rotationMode=BMM_Additive
    rotationSpace=BCS_BoneSpace
    translationMode=BMM_Additive
    translationSpace=BCS_BoneSpace
    alphaBoolBlend {blendOption=Linear}
    boneToModify {boneName=base_hair_01_r}
    rotation {yaw=-5}
    translation {y=-1, z=1}
   }
   n12 = Get bIsTargeting()
   n13 = "Transform (Modify) Bone - Bone: base_hair_01_r"(ComponentPose=@n11, bAlphaBoolEnabled=@n12, Translation="0,1,0", Rotation="-5,0,0", Scale="1,1,1") {
    alphaInputType=Bool
    rotationMode=BMM_Additive
    rotationSpace=BCS_BoneSpace
    translationMode=BMM_Additive
    translationSpace=BCS_BoneSpace
    alphaBoolBlend {blendOption=Cubic}
    boneToModify {boneName=base_hair_01_r}
    rotation {pitch=-5}
    translation {y=1}
   }
   n14 = "Transform (Modify) Bone - Bone: dyn_hair_back_01_r"!(ComponentPose=@n13, bAlphaBoolEnabled:bool=@n12, Translation="0,0,1", Rotation="2,0,0", Scale="1,1,1") -> (Pose) {
    alphaInputType=Bool
    rotationMode=BMM_Additive
    rotationSpace=BCS_BoneSpace
    translationMode=BMM_Additive
    translationSpace=BCS_BoneSpace
    alphaBoolBlend {blendOption=Cubic}
    boneToModify {boneName=dyn_hair_back_01_r}
    rotation {pitch=2}
    translation {z=1}
   }
   n15 = Get head_rot_back_Extreme() -> (head_rot_back_Extreme:double)
   n16 = "Transform (Modify) Bone - Bone: base_hair_back_02_c"(ComponentPose=@n14, Alpha=@n15, Translation="3,2,0", Rotation="0,0,0", Scale="1,1,1") {
    rotationMode=BMM_Additive
    rotationSpace=BCS_BoneSpace
    translationMode=BMM_Additive
    translationSpace=BCS_BoneSpace
    alphaBoolBlend {blendOption=Linear}
    boneToModify {boneName=base_hair_back_02_c}
    translation=<3 2>
   }
   n17 = "Transform (Modify) Bone - Bone: base_hair_back_01_r"(ComponentPose=@n16, Alpha=@n15, Translation="1,0,-2", Rotation="0,-5,0", Scale="1,1,1") {
    rotationMode=BMM_Additive
    rotationSpace=BCS_BoneSpace
    translationMode=BMM_Additive
    translationSpace=BCS_BoneSpace
    alphaBoolBlend {blendOption=Linear}
    boneToModify {boneName=base_hair_back_01_r}
    rotation {yaw=-5}
    translation {x=1, z=-2}
   }
   n18 = "Transform (Modify) Bone - Bone: cape"(ComponentPose=@n17, Alpha=Get head_rot_back() -> (head_rot_back:double), Translation="2,-1.1,0", Rotation="0,12,0", Scale="1,1,1") {
    rotationMode=BMM_Additive
    rotationSpace=BCS_BoneSpace
    translationMode=BMM_Additive
    translationSpace=BCS_BoneSpace
    alphaBoolBlend {blendOption=Linear}
    boneToModify {boneName=cape}
    rotation {yaw=12}
    translation=<2 -1.1>
   }
   n19 = Get head_rot_back_neg() -> (head_rot_back_neg:double)
   n20 = "Transform (Modify) Bone - Bone: base_hair_back_01_r"(ComponentPose=@n18, Alpha=@n19, Translation="0,0,0", Rotation="45,0,0", Scale="1,1,1") {
    rotationMode=BMM_Additive
    rotationSpace=BCS_BoneSpace
    translationMode=BMM_Additive
    translationSpace=BCS_BoneSpace
    alphaBoolBlend {blendOption=Linear}
    boneToModify {boneName=base_hair_back_01_r}
    rotation {pitch=45}
   }
   n21 = "Transform (Modify) Bone - Bone: base_hair_back_02_r"(ComponentPose=@n20, Alpha=@n19, Translation="0,0,0", Rotation="45,0,0", Scale="1,1,1") {
    rotationMode=BMM_Additive
    rotationSpace=BCS_BoneSpace
    translationMode=BMM_Additive
    translationSpace=BCS_BoneSpace
    alphaBoolBlend {blendOption=Linear}
    boneToModify {boneName=base_hair_back_02_r}
    rotation {pitch=45}
   }
   n22 = "Transform (Modify) Bone - Bone: base_hair_back_01_c"(ComponentPose=@n21, Alpha=@n19, Translation="0,0,0", Rotation="45,0,0", Scale="1,1,1") {
    rotationMode=BMM_Additive
    rotationSpace=BCS_BoneSpace
    translationMode=BMM_Additive
    translationSpace=BCS_BoneSpace
    alphaBoolBlend {blendOption=Linear}
    boneToModify {boneName=base_hair_back_01_c}
    rotation {pitch=45}
   }
   n23 = "Transform (Modify) Bone - Bone: base_hair_back_02_c"(ComponentPose=@n22, Alpha=@n19, Translation="0,0,0", Rotation="45,0,0", Scale="1,1,1") {
    rotationMode=BMM_Additive
    rotationSpace=BCS_BoneSpace
    translationMode=BMM_Additive
    translationSpace=BCS_BoneSpace
    alphaBoolBlend {blendOption=Linear}
    boneToModify {boneName=base_hair_back_02_c}
    rotation {pitch=45}
   }
   n24 = "Transform (Modify) Bone - Bone: base_hair_back_01_l"(ComponentPose=@n23, Alpha=@n19, Translation="0,0,0", Rotation="45,0,0", Scale="1,1,1") {
    rotationMode=BMM_Additive
    rotationSpace=BCS_BoneSpace
    translationMode=BMM_Additive
    translationSpace=BCS_BoneSpace
    alphaBoolBlend {blendOption=Linear}
    boneToModify {boneName=base_hair_back_01_l}
    rotation {pitch=45}
   }
   n25 = "Transform (Modify) Bone - Bone: base_hair_back_02_l"(ComponentPose=@n24, Alpha=@n19, Translation="0,0,0", Rotation="45,0,0", Scale="1,1,1") {
    rotationMode=BMM_Additive
    rotationSpace=BCS_BoneSpace
    translationMode=BMM_Additive
    translationSpace=BCS_BoneSpace
    alphaBoolBlend {blendOption=Linear}
    boneToModify {boneName=base_hair_back_02_l}
    rotation {pitch=45}
   }
   n26 = Control Rig!(Source=Component To Local(ComponentPose=@n25), Alpha, onlypistol:bool=Get Only_Pistol() -> (Only_Pistol:bool)) -> (Pose) {
    alphaBoolBlend {blendOption=Linear}
    controlRigAssetReference {blueprintRigClass=&/F_MED_PartyJelly_FaceAcc_front_hair_CtrlRig.F_MED_PartyJelly_FaceAcc_front_hair_CtrlRig_C}
   }
   n27 = RigidBody!(ComponentPose=Local To Component(LocalPose=@n26), Alpha, OverrideWorldGravity:Vector=Get GravityOverride(), ExternalForce) -> (Pose) {
    bOverrideWorldGravity=true
    overridePhysicsAsset=&/F_MED_PartyJelly_FaceAcc_Physics
    simulationSpace=BaseBoneSpace
    alphaBoolBlend {blendOption=Linear}
    baseBoneRef {boneName=head}
    overrideWorldGravity {z=-980}
    simSpaceSettings {maxAngularAcceleration=40, maxAngularVelocity=8, maxLinearAcceleration=2000, maxLinearVelocity=500, worldAlpha=1}
   }
   n28 = Get isWearingBackbling()
   n29 = "Constraint - dyn_hair_back_02_c (PPP)"!(ComponentPose=@n27, bAlphaBoolEnabled:bool=@n28, ConstraintWeights_0=0.6, ConstraintWeights_1:float=0.2, ConstraintWeights_2:float=0.2) -> (Pose) {
    alphaInputType=Bool
    alphaBoolBlend {blendOption=Linear}
    boneToModify {boneName=dyn_hair_back_02_c}
    constraintSetup [
     {perAxis {bZ=false}, targetBone {boneName=spine_05}}
     {perAxis {bZ=false}, targetBone {boneName=base_hair_back_03_r}}
     {targetBone {boneName=base_hair_back_03_l}}
    ]
    constraintWeights [1,1,1]
   }
   n30 = "Constraint - dyn_hair_back_03_c (PPP)"!(ComponentPose=@n29, bAlphaBoolEnabled:bool=@n28, ConstraintWeights_0=0.6, ConstraintWeights_1:float=0.2, ConstraintWeights_2:float=0.2) -> (Pose) {
    alphaInputType=Bool
    alphaBoolBlend {blendOption=Linear}
    boneToModify {boneName=dyn_hair_back_03_c}
    constraintSetup [
     {perAxis {bZ=false}, targetBone {boneName=spine_05}}
     {perAxis {bZ=false}, targetBone {boneName=base_hair_back_03_r}}
     {targetBone {boneName=base_hair_back_03_l}}
    ]
    constraintWeights [1,1,1]
   }
   n31 = Get wearing_backpack() -> (wearing_backpack:bool)
   n32 = "Transform (Modify) Bone - Bone: base_hair_back_02_c"!(ComponentPose=@n30, bAlphaBoolEnabled:bool=@n31, Translation="0,0,0", Rotation="-60,0,0", Scale="1,1,1") -> (Pose) {
    alphaInputType=Bool
    rotationMode=BMM_Additive
    rotationSpace=BCS_BoneSpace
    translationMode=BMM_Additive
    translationSpace=BCS_BoneSpace
    alphaBoolBlend {blendOption=Linear}
    boneToModify {boneName=base_hair_back_02_c}
    rotation {pitch=-60}
   }
   n33 = "Transform (Modify) Bone - Bone: base_hair_back_03_c"!(ComponentPose=@n32, bAlphaBoolEnabled:bool=@n31, Translation="0,0,0", Rotation="-10,0,0", Scale="1,1,1") -> (Pose) {
    alphaInputType=Bool
    rotationMode=BMM_Additive
    rotationSpace=BCS_BoneSpace
    translationMode=BMM_Additive
    translationSpace=BCS_BoneSpace
    alphaBoolBlend {blendOption=Linear}
    boneToModify {boneName=base_hair_back_03_c}
    rotation {pitch=-10}
   }
   n34 = "Constraint - dyn_hair_01_r (R)"(ComponentPose=@n33, Alpha=0.6, ConstraintWeights_0=1) {
    alphaBoolBlend {blendOption=Linear}
    boneToModify {boneName=dyn_hair_01_r}
    constraintSetup [{transformType=Rotation, targetBone {boneName=base_hair_back_01_r}}]
    constraintWeights [1]
   }
   n35 = "Transform (Modify) Bone - Bone: base_hair_01_r"!(ComponentPose=@n34, Alpha:float=Get Head Left() -> (Head Left:double), Translation="0,-2,0", Rotation="0,0,0", Scale="1,1,1") -> (Pose) {
    rotationMode=BMM_Additive
    rotationSpace=BCS_BoneSpace
    translationMode=BMM_Additive
    translationSpace=BCS_BoneSpace
    alphaBoolBlend {blendOption=Linear}
    alphaScaleBiasClamp {bInterpResult=true, interpSpeedDecreasing=4, interpSpeedIncreasing=4}
    boneToModify {boneName=base_hair_01_r}
    translation {y=-2}
   }
   n36 = "Transform (Modify) Bone - Bone: base_hair_01_r"(ComponentPose=@n35, Alpha=Get Head Right() -> (Head Right:double), Translation="0,-2,0", Rotation="0,0,0", Scale="1,1,1") {
    rotationMode=BMM_Additive
    rotationSpace=BCS_BoneSpace
    translationMode=BMM_Additive
    translationSpace=BCS_BoneSpace
    alphaBoolBlend {blendOption=Linear}
    alphaScaleBiasClamp {bInterpResult=true, interpSpeedDecreasing=4, interpSpeedIncreasing=4}
    boneToModify {boneName=base_hair_01_r}
    translation {y=-2}
   }
   n37 = Control Rig!(
    Source=Component To Local(ComponentPose=@n36)
    Alpha
    Only_Rifle:bool=Get Only_Rifle() -> (Only_Rifle:bool)
    Pistol_Crouch:bool=Get Pistol_Crouch() -> (Pistol_Crouch:bool)
    wearing_backpack:bool=Get wearing_backpack_jump() -> (wearing_backpack_jump:bool)
   ) -> (Pose) {
    alphaBoolBlend {blendOption=Linear}
    controlRigAssetReference {blueprintRigClass=&/F_MED_PartyJelly_FaceAcc_front_Blob_CtrlRig.F_MED_PartyJelly_FaceAcc_front_Blob_CtrlRig_C}
   }
   n38 = "Transform (Modify) Bone - Bone: base_hair_01_r"(
    ComponentPose=Local To Component(LocalPose=@n37)
    Alpha=Get crouch_Walking() -> (crouch_Walking:double)
    Translation="0,2,0"
    Rotation="0,10,0"
    Scale="1,1,1"
   ) {
    rotationMode=BMM_Additive
    rotationSpace=BCS_BoneSpace
    translationMode=BMM_Additive
    translationSpace=BCS_BoneSpace
    alphaBoolBlend {blendOption=Linear}
    alphaScaleBiasClamp {bInterpResult=true}
    boneToModify {boneName=base_hair_01_r}
    rotation {yaw=10}
    translation {y=2}
   }
   n39 = "Transform (Modify) Bone - Bone: base_hair_01_r"(ComponentPose=@n38, Alpha=Get Head Down() -> (Head Down:double), Translation="0,1,0", Rotation="0,-15,0", Scale="1,1,1") {
    rotationMode=BMM_Additive
    rotationSpace=BCS_BoneSpace
    translationMode=BMM_Additive
    translationSpace=BCS_BoneSpace
    alphaBoolBlend {blendOption=Linear}
    alphaScaleBiasClamp {bInterpResult=true}
    boneToModify {boneName=base_hair_01_r}
    rotation {yaw=-15}
    translation {y=1}
   }
   n40 = Get GravityOverride()
   n41 = Anim Dynamics - Bone: dyn_hair_front_01_r!(ComponentPose=@n39, Alpha=1, GravityOverride:Vector=@n40, ExternalForce="0,0,0") -> (Pose) {
    angularDampingOverride=1
    angularSpringConstant=20000
    bAngularSpring=true
    bOverrideAngularDamping=true
    bUseGravityOverride=true
    bUseSphericalLimits=true
    gravityScale=0.35
    simulationSpace=RootRelative
    alphaBoolBlend {blendOption=Linear}
    boundBone {boneName=dyn_hair_front_01_r}
    componentLinearAccScale=<1 1 1>
    componentLinearVelScale=<1 1 1>
    gravityOverride {z=-980}
    physicsBodyDefinitions [{
      boundBone {boneName=dyn_hair_front_01_r}
      boxExtents=<20 30 30>
      constraintSetup {angularLimitsMax=<2 2 2>, angularLimitsMin=<-2 -2 -2>, angularTarget=<1 0 0>}
      localJointOffset=<4 0 0>
     }]
   }
   n42 = Anim Dynamics - Bone: dyn_hair_front_02_r!(ComponentPose=@n41, Alpha=1, GravityOverride:Vector=@n40, ExternalForce="0,0,0") -> (Pose) {
    angularDampingOverride=1
    angularSpringConstant=10000
    bAngularSpring=true
    bOverrideAngularDamping=true
    bUseGravityOverride=true
    bUseSphericalLimits=true
    gravityScale=0.35
    simulationSpace=RootRelative
    alphaBoolBlend {blendOption=Linear}
    boundBone {boneName=dyn_hair_front_02_r}
    componentLinearAccScale=<1 1 1>
    componentLinearVelScale=<1 1 1>
    gravityOverride {z=-980}
    physicsBodyDefinitions [{
      boundBone {boneName=dyn_hair_front_02_r}
      boxExtents=<20 30 30>
      constraintSetup {angularLimitsMax=<10 10 10>, angularLimitsMin=<-10 -10 -10>, angularTarget=<1 0 0>}
      localJointOffset=<4 0 0>
     }]
    sphericalLimits [{limitRadius=8, drivingBone {boneName=upperarm_r}, sphereLocalOffset=<0 0 0>}]
   }
   n43 = Anim Dynamics - Bone: dyn_hair_front_03_r!(ComponentPose=@n42, Alpha=1, GravityOverride:Vector=@n40, ExternalForce="0,0,0") -> (Pose) {
    angularDampingOverride=1
    angularSpringConstant=10000
    bAngularSpring=true
    bOverrideAngularDamping=true
    bUseGravityOverride=true
    bUsePlanarLimit=false
    bUseSphericalLimits=true
    gravityScale=0.35
    simulationSpace=RootRelative
    alphaBoolBlend {blendOption=Linear}
    boundBone {boneName=dyn_hair_front_03_r}
    componentLinearAccScale=<1 1 1>
    componentLinearVelScale=<1 1 1>
    gravityOverride {z=-980}
    physicsBodyDefinitions [{
      boundBone {boneName=dyn_hair_front_03_r}
      boxExtents=<20 30 30>
      constraintSetup {angularLimitsMax=<15 15 15>, angularLimitsMin=<-15 -15 -15>, angularTarget=<1 0 0>}
      localJointOffset=<4 0 0>
     }]
    sphericalLimits [
     {limitRadius=10, drivingBone {boneName=clavicle_r}, sphereLocalOffset=<-5.2979 1.0461 1.7568>}
     {limitRadius=8, drivingBone {boneName=neck_02}, sphereLocalOffset=<-3.3594 4.5904 0.4087>}
     {limitRadius=8, drivingBone {boneName=upperarm_r}, sphereLocalOffset=<-9.7212 0.1042 -0.7702>}
     {limitRadius=0, drivingBone {boneName=None}, sphereLocalOffset=<0 0 0>}
     {limitRadius=0, drivingBone {boneName=None}, sphereLocalOffset=<0 0 0>}
    ]
   }
   n44 = Anim Dynamics - Bone: dyn_hair_front_01_l!(ComponentPose=@n43, Alpha=1, GravityOverride:Vector=@n40, ExternalForce="0,0,0") -> (Pose) {
    angularDampingOverride=1
    angularSpringConstant=10000
    bAngularSpring=true
    bOverrideAngularDamping=true
    bUseGravityOverride=true
    gravityScale=0.35
    simulationSpace=RootRelative
    alphaBoolBlend {blendOption=Linear}
    boundBone {boneName=dyn_hair_front_01_l}
    componentLinearAccScale=<1 1 1>
    componentLinearVelScale=<1 1 1>
    gravityOverride {z=-980}
    physicsBodyDefinitions [{
      boundBone {boneName=dyn_hair_front_01_l}
      boxExtents=<20 30 30>
      constraintSetup {angularLimitsMax=<5 5 5>, angularLimitsMin=<-5 -5 -5>, angularTarget=<1 0 0>}
      localJointOffset=<4 0 0>
     }]
   }
   n45 = Anim Dynamics - Bone: dyn_hair_front_02_l!(ComponentPose=@n44, Alpha=1, GravityOverride:Vector=@n40, ExternalForce="0,0,0") -> (Pose) {
    angularDampingOverride=1
    angularSpringConstant=10000
    bAngularSpring=true
    bOverrideAngularDamping=true
    bUseGravityOverride=true
    bUseSphericalLimits=true
    gravityScale=0.35
    simulationSpace=RootRelative
    alphaBoolBlend {blendOption=Linear}
    boundBone {boneName=dyn_hair_front_02_l}
    componentLinearAccScale=<1 1 1>
    componentLinearVelScale=<1 1 1>
    gravityOverride {z=-980}
    physicsBodyDefinitions [{
      boundBone {boneName=dyn_hair_front_02_l}
      boxExtents=<20 30 30>
      constraintSetup {angularLimitsMax=<10 10 10>, angularLimitsMin=<-10 -10 -10>, angularTarget=<1 0 0>}
      localJointOffset=<4 0 0>
     }]
    sphericalLimits [{limitRadius=8, drivingBone {boneName=upperarm_l}, sphereLocalOffset=<0 0 0>}]
   }
   n46 = Anim Dynamics - Bone: dyn_hair_front_03_l!(ComponentPose=@n45, Alpha=1, GravityOverride:Vector=@n40, ExternalForce="0,0,0") -> (Pose) {
    angularDampingOverride=1
    angularSpringConstant=10000
    bAngularSpring=true
    bOverrideAngularDamping=true
    bUseGravityOverride=true
    bUsePlanarLimit=false
    bUseSphericalLimits=true
    gravityScale=0.35
    simulationSpace=RootRelative
    alphaBoolBlend {blendOption=Linear}
    boundBone {boneName=dyn_hair_front_03_l}
    componentLinearAccScale=<1 1 1>
    componentLinearVelScale=<1 1 1>
    gravityOverride {z=-980}
    physicsBodyDefinitions [{
      boundBone {boneName=dyn_hair_front_03_l}
      boxExtents=<20 30 30>
      constraintSetup {angularLimitsMax=<15 15 15>, angularLimitsMin=<-15 -15 -15>, angularTarget=<1 0 0>}
      localJointOffset=<0.846 0 -0.1392>
     }]
    planarLimits [{drivingBone {boneName=upperarm_l}, planeTransform {location=<-3.6809 7.3298 4.1544>, rotation=<rot 50 -179.9996 -179.9997>}}]
    sphericalLimits [
     {limitRadius=10, drivingBone {boneName=clavicle_l}, sphereLocalOffset=<5.9204 0 -2.0495>}
     {limitRadius=8, drivingBone {boneName=neck_02}, sphereLocalOffset=<-3.3594 4.5904 0.4087>}
    ]
   }
   n47 = Get FrontEnd_not_emote()
   n48 = "Transform (Modify) Bone - Bone: base_hair_02_l"!(ComponentPose=@n46, bAlphaBoolEnabled:bool=@n47, Translation="0,0,0", Rotation="-15,0,0", Scale="1,1,1") -> (Pose) {
    alphaInputType=Bool
    rotationMode=BMM_Additive
    rotationSpace=BCS_BoneSpace
    translationMode=BMM_Additive
    translationSpace=BCS_BoneSpace
    alphaBoolBlend {blendOption=Cubic}
    boneToModify {boneName=base_hair_02_l}
    rotation {pitch=-15}
   }
   n49 = "Transform (Modify) Bone - Bone: base_hair_03_l"!(ComponentPose=@n48, bAlphaBoolEnabled:bool=@n47, Translation="0,0,0", Rotation="-5,0,0", Scale="1,1,1") -> (Pose) {
    alphaInputType=Bool
    rotationMode=BMM_Additive
    rotationSpace=BCS_BoneSpace
    translationMode=BMM_Additive
    translationSpace=BCS_BoneSpace
    alphaBoolBlend {blendOption=Cubic}
    boneToModify {boneName=base_hair_03_l}
    rotation {pitch=-5}
   }
   n50 = "Transform (Modify) Bone - Bone: base_hair_03_r"!(ComponentPose=@n49, bAlphaBoolEnabled:bool=@n47, Translation="0,0,0", Rotation="-20,0,0", Scale="1,1,1") -> (Pose) {
    alphaInputType=Bool
    rotationMode=BMM_Additive
    rotationSpace=BCS_BoneSpace
    translationMode=BMM_Additive
    translationSpace=BCS_BoneSpace
    alphaBoolBlend {blendOption=Linear}
    boneToModify {boneName=base_hair_03_r}
    rotation {pitch=-20}
   }
   n51 = "Transform (Modify) Bone - Bone: base_hair_01_l"!(ComponentPose=@n50, bAlphaBoolEnabled:bool=@n47, Translation="0,0,0", Rotation="8,12,0", Scale="1,1,1") -> (Pose) {
    alphaInputType=Bool
    rotationMode=BMM_Additive
    rotationSpace=BCS_BoneSpace
    translationMode=BMM_Additive
    translationSpace=BCS_BoneSpace
    alphaBoolBlend {blendOption=Linear}
    boneToModify {boneName=base_hair_01_l}
    rotation {pitch=8, yaw=12}
   }
   n52 = "Transform (Modify) Bone - Bone: base_hair_01_r"!(ComponentPose=@n51, bAlphaBoolEnabled:bool=@n47, Translation="0,0,0", Rotation="10,5,0", Scale="1,1,1") -> (Pose) {
    alphaInputType=Bool
    rotationMode=BMM_Additive
    rotationSpace=BCS_BoneSpace
    translationMode=BMM_Additive
    translationSpace=BCS_BoneSpace
    alphaBoolBlend {blendOption=Linear}
    boneToModify {boneName=base_hair_01_r}
    rotation {pitch=10, yaw=5}
   }
   n53 = Get FrontEnd_Emote() -> (FrontEnd_Emote:bool)
   n54 = "Transform (Modify) Bone - Bone: base_hair_back_01_r"!(ComponentPose=@n52, bAlphaBoolEnabled:bool=@n53, Translation="0,0,0", Rotation="10,0,0", Scale="1,1,1") -> (Pose) {
    alphaInputType=Bool
    rotationMode=BMM_Additive
    rotationSpace=BCS_BoneSpace
    translationMode=BMM_Additive
    translationSpace=BCS_BoneSpace
    alphaBoolBlend {blendOption=Linear}
    boneToModify {boneName=base_hair_back_01_r}
    rotation {pitch=10}
   }
   n55 = "Transform (Modify) Bone - Bone: base_hair_01_r"(ComponentPose=@n54, bAlphaBoolEnabled=@n53, Translation="0,0,0", Rotation="15,0,0", Scale="1,1,1") {
    alphaInputType=Bool
    rotationMode=BMM_Additive
    rotationSpace=BCS_BoneSpace
    translationMode=BMM_Additive
    translationSpace=BCS_BoneSpace
    alphaBoolBlend {blendOption=Linear}
    boneToModify {boneName=base_hair_01_r}
    rotation {pitch=15}
   }
   n56 = "Transform (Modify) Bone - Bone: base_hair_back_01_c"!(ComponentPose=@n55, bAlphaBoolEnabled:bool=@n53, Translation="0,0,0", Rotation="10,0,0", Scale="1,1,1") -> (Pose) {
    alphaInputType=Bool
    rotationMode=BMM_Additive
    rotationSpace=BCS_BoneSpace
    translationMode=BMM_Additive
    translationSpace=BCS_BoneSpace
    alphaBoolBlend {blendOption=Linear}
    boneToModify {boneName=base_hair_back_01_c}
    rotation {pitch=10}
   }
   n57 = "Transform (Modify) Bone - Bone: base_hair_back_01_l"!(ComponentPose=@n56, bAlphaBoolEnabled:bool=@n53, Translation="0,0,0", Rotation="10,0,0", Scale="1,1,1") -> (Pose) {
    alphaInputType=Bool
    rotationMode=BMM_Additive
    rotationSpace=BCS_BoneSpace
    translationMode=BMM_Additive
    translationSpace=BCS_BoneSpace
    alphaBoolBlend {blendOption=Linear}
    boneToModify {boneName=base_hair_back_01_l}
    rotation {pitch=10}
   }
   n58 = "Transform (Modify) Bone - Bone: base_hair_01_l"(ComponentPose=@n57, bAlphaBoolEnabled=@n53, Translation="0,0,0", Rotation="15,0,0", Scale="1,1,1") {
    alphaInputType=Bool
    rotationMode=BMM_Additive
    rotationSpace=BCS_BoneSpace
    translationMode=BMM_Additive
    translationSpace=BCS_BoneSpace
    alphaBoolBlend {blendOption=Linear}
    boneToModify {boneName=base_hair_01_l}
    rotation {pitch=15}
   }
   n59 = "Transform (Modify) Bone - Bone: dyn_hair_01_r"!(ComponentPose=@n58, bAlphaBoolEnabled:bool=True, Translation="0,0,0", Rotation="-15,-5,0", Scale="1,1,1") -> (Pose) {
    alphaInputType=Bool
    rotationMode=BMM_Additive
    rotationSpace=BCS_BoneSpace
    translationMode=BMM_Additive
    translationSpace=BCS_BoneSpace
    alphaBoolBlend {blendOption=Linear}
    boneToModify {boneName=dyn_hair_01_r}
    rotation {pitch=-15, yaw=-5}
   }
   n60 = "Transform (Modify) Bone - Bone: dyn_hair_02_r"!(ComponentPose=@n59, bAlphaBoolEnabled:bool=True, Translation="0,0,0", Rotation="10,0,0", Scale="1,1,1") -> (Pose) {
    alphaInputType=Bool
    rotationMode=BMM_Additive
    rotationSpace=BCS_BoneSpace
    translationMode=BMM_Additive
    translationSpace=BCS_BoneSpace
    alphaBoolBlend {blendOption=Linear}
    boneToModify {boneName=dyn_hair_02_r}
    rotation {pitch=10}
   }
   n61 = "Transform (Modify) Bone - Bone: dyn_hair_03_r"!(ComponentPose=@n60, bAlphaBoolEnabled:bool=True, Translation="0,0,0", Rotation="0,0,0", Scale="1,1,1") -> (Pose) {
    alphaInputType=Bool
    rotationMode=BMM_Additive
    rotationSpace=BCS_BoneSpace
    translationMode=BMM_Additive
    translationSpace=BCS_BoneSpace
    alphaBoolBlend {blendOption=Linear}
    boneToModify {boneName=dyn_hair_03_r}
   }
   Component To Local(ComponentPose=@n61)
   Component To Local()
   n62 = "Constraint - dyn_hair_back_01_l (PPP)"!(ComponentPose, bAlphaBoolEnabled:bool=True, ConstraintWeights_0=0.5, ConstraintWeights_1:float=0.2, ConstraintWeights_2:float=0.3) -> (Pose) {
    alphaInputType=Bool
    alphaBoolBlend {blendOption=Linear}
    boneToModify {boneName=dyn_hair_back_01_l}
    constraintSetup [
     {perAxis {bZ=false}, targetBone {boneName=spine_05}}
     {perAxis {bZ=false}, targetBone {boneName=dyn_hair_back_02_c}}
     {targetBone {boneName=head}}
    ]
    constraintWeights [1,1,1]
   }
   n63 = "Constraint - dyn_hair_back_02_l (PPP)"!(ComponentPose=@n62, bAlphaBoolEnabled:bool=True, ConstraintWeights_0=0.5, ConstraintWeights_1:float=0.2, ConstraintWeights_2:float=0.3) -> (Pose) {
    alphaInputType=Bool
    alphaBoolBlend {blendOption=Linear}
    boneToModify {boneName=dyn_hair_back_02_l}
    constraintSetup [
     {perAxis {bZ=false}, targetBone {boneName=spine_05}}
     {perAxis {bZ=false}, targetBone {boneName=dyn_hair_back_02_c}}
     {targetBone {boneName=head}}
    ]
    constraintWeights [1,1,1]
   }
   n64 = "Constraint - dyn_hair_back_03_l (PPP)"!(ComponentPose=@n63, bAlphaBoolEnabled:bool=True, ConstraintWeights_0=0.5, ConstraintWeights_1:float=0.2, ConstraintWeights_2:float=0.3) -> (Pose) {
    alphaInputType=Bool
    alphaBoolBlend {blendOption=Linear}
    boneToModify {boneName=dyn_hair_back_03_l}
    constraintSetup [
     {perAxis {bZ=false}, targetBone {boneName=spine_05}}
     {perAxis {bZ=false}, targetBone {boneName=dyn_hair_back_03_c}}
     {targetBone {boneName=head}}
    ]
    constraintWeights [1,1,1]
   }
   n65 = "Constraint - dyn_hair_01_l (PPP)"!(ComponentPose=@n64, bAlphaBoolEnabled:bool=True, ConstraintWeights_0=0.6, ConstraintWeights_1:float=0.2, ConstraintWeights_2:float=0.2) -> (Pose) {
    alphaInputType=Bool
    alphaBoolBlend {blendOption=Linear}
    boneToModify {boneName=dyn_hair_01_l}
    constraintSetup [
     {perAxis {bZ=false}, targetBone {boneName=spine_05}}
     {perAxis {bZ=false}, targetBone {boneName=dyn_hair_back_02_l}}
     {targetBone {boneName=clavicle_l}}
    ]
    constraintWeights [1,1,1]
   }
   n66 = "Constraint - dyn_hair_02_l (PPP)"!(ComponentPose=@n65, bAlphaBoolEnabled:bool=True, ConstraintWeights_0=0.6, ConstraintWeights_1:float=0.2, ConstraintWeights_2:float=0.2) -> (Pose) {
    alphaInputType=Bool
    alphaBoolBlend {blendOption=Linear}
    boneToModify {boneName=dyn_hair_02_l}
    constraintSetup [
     {perAxis {bZ=false}, targetBone {boneName=spine_05}}
     {perAxis {bZ=false}, targetBone {boneName=dyn_hair_back_02_l}}
     {targetBone {boneName=clavicle_l}}
    ]
    constraintWeights [1,1,1]
   }
   "Constraint - dyn_hair_03_l (PPP)"!(ComponentPose=@n66, bAlphaBoolEnabled:bool=True, ConstraintWeights_0=0.6, ConstraintWeights_1:float=0.2, ConstraintWeights_2:float=0.2) -> (Pose) {
    alphaInputType=Bool
    alphaBoolBlend {blendOption=Linear}
    boneToModify {boneName=dyn_hair_03_l}
    constraintSetup [
     {perAxis {bZ=false}, targetBone {boneName=spine_05}}
     {perAxis {bZ=false}, targetBone {boneName=dyn_hair_back_03_l}}
     {targetBone {boneName=clavicle_l}}
    ]
    constraintWeights [1,1,1]
   }
   "Constraint - dyn_hair_back_01_r (PPP)"!(ComponentPose, bAlphaBoolEnabled:bool=True, ConstraintWeights_0=0.5, ConstraintWeights_1:float=0.2, ConstraintWeights_2:float=0.3) -> (Pose) {
    alphaInputType=Bool
    alphaBoolBlend {blendOption=Linear}
    boneToModify {boneName=dyn_hair_back_01_r}
    constraintSetup [
     {perAxis {bZ=false}, targetBone {boneName=spine_05}}
     {perAxis {bZ=false}, targetBone {boneName=dyn_hair_back_02_c}}
     {targetBone {boneName=head}}
    ]
    constraintWeights [1,1,1]
   }
   n67 = "Constraint - dyn_hair_01_l (PPPP)"!(
    ComponentPose
    bAlphaBoolEnabled:bool=True
    ConstraintWeights_0=0.3
    ConstraintWeights_1:float=0.3
    ConstraintWeights_2:float=0.2
    ConstraintWeights_3:float=0.2
   ) -> (Pose) {
    alphaInputType=Bool
    alphaBoolBlend {blendOption=Linear}
    boneToModify {boneName=dyn_hair_01_l}
    constraintSetup [
     {perAxis {bZ=false}, targetBone {boneName=spine_05}}
     {perAxis {bZ=false}, targetBone {boneName=clavicle_r}}
     {targetBone {boneName=dyn_hair_back_02_c}}
     {targetBone {boneName=head}}
    ]
    constraintWeights [1,1,1,1]
   }
   n68 = "Constraint - dyn_hair_02_l (PPPP)"!(
    ComponentPose=@n67
    bAlphaBoolEnabled:bool=True
    ConstraintWeights_0=0.3
    ConstraintWeights_1:float=0.3
    ConstraintWeights_2:float=0.2
    ConstraintWeights_3:float=0.2
   ) -> (Pose) {
    alphaInputType=Bool
    alphaBoolBlend {blendOption=Linear}
    boneToModify {boneName=dyn_hair_02_l}
    constraintSetup [
     {perAxis {bZ=false}, targetBone {boneName=spine_05}}
     {perAxis {bZ=false}, targetBone {boneName=clavicle_r}}
     {targetBone {boneName=dyn_hair_back_02_c}}
     {targetBone {boneName=head}}
    ]
    constraintWeights [1,1,1,1]
   }
   "Constraint - dyn_hair_03_l (PPPP)"!(
    ComponentPose=@n68
    bAlphaBoolEnabled:bool=True
    ConstraintWeights_0=0.3
    ConstraintWeights_1:float=0.3
    ConstraintWeights_2:float=0.2
    ConstraintWeights_3:float=0.2
   ) -> (Pose) {
    alphaInputType=Bool
    alphaBoolBlend {blendOption=Linear}
    boneToModify {boneName=dyn_hair_03_l}
    constraintSetup [
     {perAxis {bZ=false}, targetBone {boneName=spine_05}}
     {perAxis {bZ=false}, targetBone {boneName=clavicle_r}}
     {targetBone {boneName=dyn_hair_back_03_c}}
     {targetBone {boneName=head}}
    ]
    constraintWeights [1,1,1,1]
   }
   n69 = "Constraint - dyn_hair_01_r (PPPP)"!(
    ComponentPose
    bAlphaBoolEnabled:bool=True
    ConstraintWeights_0=0.3
    ConstraintWeights_1:float=0.3
    ConstraintWeights_2:float=0.2
    ConstraintWeights_3:float=0.2
   ) -> (Pose) {
    alphaInputType=Bool
    alphaBoolBlend {blendOption=Linear}
    boneToModify {boneName=dyn_hair_01_r}
    constraintSetup [
     {perAxis {bZ=false}, targetBone {boneName=spine_05}}
     {perAxis {bZ=false}, targetBone {boneName=dyn_hair_back_02_r}}
     {targetBone {boneName=head}}
     {targetBone {boneName=head}}
    ]
    constraintWeights [1,1,1,1]
   }
   n70 = "Constraint - dyn_hair_02_r (PPPP)"!(
    ComponentPose=@n69
    bAlphaBoolEnabled:bool=True
    ConstraintWeights_0=0.3
    ConstraintWeights_1:float=0.3
    ConstraintWeights_2:float=0.2
    ConstraintWeights_3:float=0.2
   ) -> (Pose) {
    alphaInputType=Bool
    alphaBoolBlend {blendOption=Linear}
    boneToModify {boneName=dyn_hair_02_r}
    constraintSetup [
     {perAxis {bZ=false}, targetBone {boneName=spine_05}}
     {perAxis {bZ=false}, targetBone {boneName=clavicle_r}}
     {targetBone {boneName=dyn_hair_back_02_r}}
     {targetBone {boneName=head}}
    ]
    constraintWeights [1,1,1,1]
   }
   "Constraint - dyn_hair_03_r (PPPP)"!(
    ComponentPose=@n70
    bAlphaBoolEnabled:bool=True
    ConstraintWeights_0=0.3
    ConstraintWeights_1:float=0.3
    ConstraintWeights_2:float=0.2
    ConstraintWeights_3:float=0.2
   ) -> (Pose) {
    alphaInputType=Bool
    alphaBoolBlend {blendOption=Linear}
    boneToModify {boneName=dyn_hair_03_r}
    constraintSetup [
     {perAxis {bZ=false}, targetBone {boneName=spine_05}}
     {perAxis {bZ=false}, targetBone {boneName=clavicle_r}}
     {targetBone {boneName=dyn_hair_back_03_r}}
     {targetBone {boneName=head}}
    ]
    constraintWeights [1,1,1,1]
   }
   n71 = "Constraint - dyn_hair_back_01_r (PPPP)"!(
    ComponentPose
    bAlphaBoolEnabled:bool=True
    ConstraintWeights_0=0.3
    ConstraintWeights_1:float=0.3
    ConstraintWeights_2:float=0.2
    ConstraintWeights_3:float=0.2
   ) -> (Pose) {
    alphaInputType=Bool
    alphaBoolBlend {blendOption=Linear}
    boneToModify {boneName=dyn_hair_back_01_r}
    constraintSetup [
     {perAxis {bZ=false}, targetBone {boneName=spine_05}}
     {perAxis {bZ=false}, targetBone {boneName=clavicle_r}}
     {targetBone {boneName=dyn_hair_back_02_c}}
     {targetBone {boneName=head}}
    ]
    constraintWeights [1,1,1,1]
   }
   n72 = "Constraint - dyn_hair_back_02_r (PPPP)"!(
    ComponentPose=@n71
    bAlphaBoolEnabled:bool=True
    ConstraintWeights_0=0.3
    ConstraintWeights_1:float=0.3
    ConstraintWeights_2:float=0.2
    ConstraintWeights_3:float=0.2
   ) -> (Pose) {
    alphaInputType=Bool
    alphaBoolBlend {blendOption=Linear}
    boneToModify {boneName=dyn_hair_back_02_r}
    constraintSetup [
     {perAxis {bZ=false}, targetBone {boneName=spine_05}}
     {perAxis {bZ=false}, targetBone {boneName=clavicle_r}}
     {targetBone {boneName=dyn_hair_back_02_c}}
     {targetBone {boneName=head}}
    ]
    constraintWeights [1,1,1,1]
   }
   n73 = "Constraint - dyn_hair_back_03_r (PPPP)"!(
    ComponentPose=@n72
    bAlphaBoolEnabled:bool=True
    ConstraintWeights_0=0.3
    ConstraintWeights_1:float=0.3
    ConstraintWeights_2:float=0.2
    ConstraintWeights_3:float=0.2
   ) -> (Pose) {
    alphaInputType=Bool
    alphaBoolBlend {blendOption=Linear}
    boneToModify {boneName=dyn_hair_back_03_r}
    constraintSetup [
     {perAxis {bZ=false}, targetBone {boneName=spine_05}}
     {perAxis {bZ=false}, targetBone {boneName=clavicle_r}}
     {targetBone {boneName=dyn_hair_back_03_c}}
     {targetBone {boneName=head}}
    ]
    constraintWeights [1,1,1,1]
   }
   n74 = "Constraint - dyn_hair_01_r (PPPP)"(ComponentPose=@n73, bAlphaBoolEnabled=True, ConstraintWeights_0=0.5, ConstraintWeights_1=0.3, ConstraintWeights_2=0.2, ConstraintWeights_3=0) {
    alphaInputType=Bool
    alphaBoolBlend {blendOption=Linear}
    boneToModify {boneName=dyn_hair_01_r}
    constraintSetup [
     {perAxis {bZ=false}, targetBone {boneName=spine_05}}
     {perAxis {bZ=false}, targetBone {boneName=clavicle_r}}
     {targetBone {boneName=dyn_hair_back_02_c}}
     {targetBone {boneName=head}}
    ]
    constraintWeights [1,1,1,1]
   }
   n75 = "Constraint - dyn_hair_02_r (PPPP)"(ComponentPose=@n74, bAlphaBoolEnabled=True, ConstraintWeights_0=0.5, ConstraintWeights_1=0.3, ConstraintWeights_2=0.2, ConstraintWeights_3=0) {
    alphaInputType=Bool
    alphaBoolBlend {blendOption=Linear}
    boneToModify {boneName=dyn_hair_02_r}
    constraintSetup [
     {perAxis {bZ=false}, targetBone {boneName=spine_05}}
     {perAxis {bZ=false}, targetBone {boneName=clavicle_r}}
     {targetBone {boneName=dyn_hair_back_02_c}}
     {targetBone {boneName=head}}
    ]
    constraintWeights [1,1,1,1]
   }
   "Constraint - dyn_hair_03_r (PPPP)"(ComponentPose=@n75, bAlphaBoolEnabled=True, ConstraintWeights_0=0.5, ConstraintWeights_1=0.3, ConstraintWeights_2=0.2, ConstraintWeights_3=0) {
    alphaInputType=Bool
    alphaBoolBlend {blendOption=Linear}
    boneToModify {boneName=dyn_hair_03_r}
    constraintSetup [
     {perAxis {bZ=false}, targetBone {boneName=spine_05}}
     {perAxis {bZ=false}, targetBone {boneName=clavicle_r}}
     {targetBone {boneName=dyn_hair_back_03_c}}
     {targetBone {boneName=head}}
    ]
    constraintWeights [1,1,1,1]
   }
   n76 = "Constraint - dyn_hair_01_l (PPPP)"(bAlphaBoolEnabled=True, ConstraintWeights_0=0.3, ConstraintWeights_1=0.3, ConstraintWeights_2=0.2, ConstraintWeights_3=0.2) {
    alphaInputType=Bool
    alphaBoolBlend {blendOption=Linear}
    boneToModify {boneName=dyn_hair_01_l}
    constraintSetup [
     {perAxis {bZ=false}, targetBone {boneName=spine_05}}
     {perAxis {bZ=false}, targetBone {boneName=clavicle_r}}
     {targetBone {boneName=dyn_hair_back_02_c}}
     {targetBone {boneName=head}}
    ]
    constraintWeights [1,1,1,1]
   }
   n77 = "Constraint - dyn_hair_02_l (PPPP)"(
    ComponentPose=@n76
    bAlphaBoolEnabled=True
    ConstraintWeights_0=0.3
    ConstraintWeights_1=0.3
    ConstraintWeights_2=0.2
    ConstraintWeights_3=0.2
   ) {
    alphaInputType=Bool
    alphaBoolBlend {blendOption=Linear}
    boneToModify {boneName=dyn_hair_02_l}
    constraintSetup [
     {perAxis {bZ=false}, targetBone {boneName=spine_05}}
     {perAxis {bZ=false}, targetBone {boneName=clavicle_r}}
     {targetBone {boneName=dyn_hair_back_02_c}}
     {targetBone {boneName=head}}
    ]
    constraintWeights [1,1,1,1]
   }
   "Constraint - dyn_hair_03_l (PPPP)"(
    ComponentPose=@n77
    bAlphaBoolEnabled=True
    ConstraintWeights_0=0.3
    ConstraintWeights_1=0.3
    ConstraintWeights_2=0.2
    ConstraintWeights_3=0.2
   ) {
    alphaInputType=Bool
    alphaBoolBlend {blendOption=Linear}
    boneToModify {boneName=dyn_hair_03_l}
    constraintSetup [
     {perAxis {bZ=false}, targetBone {boneName=spine_05}}
     {perAxis {bZ=false}, targetBone {boneName=clavicle_r}}
     {targetBone {boneName=dyn_hair_back_03_c}}
     {targetBone {boneName=head}}
    ]
    constraintWeights [1,1,1,1]
   }
   n78 = "Transform (Modify) Bone - Bone: base_hair_back_01_r"!(ComponentPose, Alpha:float=1, Translation="0,0,0", Rotation="10,0,0", Scale="1,1,1") -> (Pose) {
    rotationMode=BMM_Additive
    rotationSpace=BCS_BoneSpace
    translationMode=BMM_Additive
    translationSpace=BCS_BoneSpace
    alphaBoolBlend {blendOption=Linear}
    alphaScaleBiasClamp {bInterpResult=true, interpSpeedDecreasing=20, interpSpeedIncreasing=9}
    boneToModify {boneName=base_hair_back_01_r}
    rotation {pitch=10}
   }
   n79 = "Transform (Modify) Bone - Bone: base_hair_back_01_c"!(ComponentPose=@n78, Alpha:float=1, Translation="0,0,0", Rotation="10,0,0", Scale="1,1,1") -> (Pose) {
    rotationMode=BMM_Additive
    rotationSpace=BCS_BoneSpace
    translationMode=BMM_Additive
    translationSpace=BCS_BoneSpace
    alphaBoolBlend {blendOption=Linear}
    alphaScaleBiasClamp {bInterpResult=true, interpSpeedDecreasing=20, interpSpeedIncreasing=9}
    boneToModify {boneName=base_hair_back_01_c}
    rotation {pitch=10}
   }
   n80 = "Transform (Modify) Bone - Bone: base_hair_back_01_l"!(ComponentPose=@n79, Alpha:float=1, Translation="0,0,0", Rotation="10,0,0", Scale="1,1,1") -> (Pose) {
    rotationMode=BMM_Additive
    rotationSpace=BCS_BoneSpace
    translationMode=BMM_Additive
    translationSpace=BCS_BoneSpace
    alphaBoolBlend {blendOption=Linear}
    alphaScaleBiasClamp {bInterpResult=true, interpSpeedDecreasing=20, interpSpeedIncreasing=9}
    boneToModify {boneName=base_hair_back_01_l}
    rotation {pitch=10}
   }
   "Transform (Modify) Bone - Bone: base_hair_01_l"!(ComponentPose=@n80, Alpha:float=1, Translation="0,0,0", Rotation="15,0,0", Scale="1,1,1") -> (Pose) {
    rotationMode=BMM_Additive
    rotationSpace=BCS_BoneSpace
    translationMode=BMM_Additive
    translationSpace=BCS_BoneSpace
    alphaBoolBlend {blendOption=Linear}
    alphaScaleBiasClamp {bInterpResult=true, interpSpeedDecreasing=20, interpSpeedIncreasing=9}
    boneToModify {boneName=base_hair_01_l}
    rotation {pitch=15}
   }
   n81 = Get Rifle_Up() -> (Rifle_Up:double)
   n82 = "Transform (Modify) Bone - Bone: dyn_hair_front_02_l"(Alpha=@n81, Translation="0,0,0", Rotation="0,15,0", Scale="1,1,1") {
    rotationMode=BMM_Additive
    rotationSpace=BCS_BoneSpace
    translationMode=BMM_Additive
    translationSpace=BCS_BoneSpace
    alphaBoolBlend {blendOption=Linear}
    boneToModify {boneName=dyn_hair_front_02_l}
    rotation {yaw=15}
   }
   "Transform (Modify) Bone - Bone: dyn_hair_front_03_l"(ComponentPose=@n82, Alpha=@n81, Translation="0,0,0", Rotation="0,15,0", Scale="1,1,1") {
    rotationMode=BMM_Additive
    rotationSpace=BCS_BoneSpace
    translationMode=BMM_Additive
    translationSpace=BCS_BoneSpace
    alphaBoolBlend {blendOption=Linear}
    boneToModify {boneName=dyn_hair_front_03_l}
    rotation {yaw=15}
   }
   n83 = "Transform (Modify) Bone - Bone: dyn_hair_front_02_l"(Alpha=1, Translation="0,0,0", Rotation="0,45,0", Scale="1,1,1") {
    rotationMode=BMM_Additive
    rotationSpace=BCS_BoneSpace
    translationMode=BMM_Additive
    translationSpace=BCS_BoneSpace
    alphaBoolBlend {blendOption=Linear}
    boneToModify {boneName=dyn_hair_front_02_l}
    rotation {yaw=45}
   }
   "Transform (Modify) Bone - Bone: dyn_hair_front_03_l"(ComponentPose=@n83, Alpha=1, Translation="0,0,0", Rotation="0,45,0", Scale="1,1,1") {
    rotationMode=BMM_Additive
    rotationSpace=BCS_BoneSpace
    translationMode=BMM_Additive
    translationSpace=BCS_BoneSpace
    alphaBoolBlend {blendOption=Linear}
    boneToModify {boneName=dyn_hair_front_03_l}
    rotation {yaw=45}
   }
   n84 = Control Rig!(Source=Use cached pose 'InitialPose'(), Alpha, FrontEnd Not Emote:bool=Get FrontEnd_not_emote()) -> (Pose) {
    alphaBoolBlend {blendOption=Linear}
    alphaScaleBiasClamp {bInterpResult=true, interpSpeedDecreasing=20, interpSpeedIncreasing=4}
    controlRigAssetReference {blueprintRigClass=&F_MED_PartyJelly_FaceAcc_CtrlRig_C}
   }
   n85 = "Transform (Modify) Bone - Bone: dyn_hair_front_02_r"(
    ComponentPose=Local To Component(LocalPose=@n84)
    Alpha=Get EmoteAnnoyingPen_Alpha() -> (EmoteAnnoyingPen_Alpha:double)
    Translation="0,0,0"
    Rotation="-5,-20,-1.54"
    Scale="1,1,1"
   ) {
    rotationMode=BMM_Additive
    rotationSpace=BCS_BoneSpace
    translationMode=BMM_Additive
    translationSpace=BCS_BoneSpace
    alphaBoolBlend {blendOption=Linear}
    alphaScaleBiasClamp {bInterpResult=true, interpSpeedIncreasing=5}
    boneToModify {boneName=dyn_hair_front_02_r}
    rotation=<rot -5 -20 -1.54>
   }
   n86 = "Transform (Modify) Bone - Bone: base_hair_01_r"(ComponentPose=@n85, bAlphaBoolEnabled=Get bIsMoving2D(), Translation="0,-1,1", Rotation="0,-5,0", Scale="1,1,1") {
    alphaInputType=Bool
    rotationMode=BMM_Additive
    rotationSpace=BCS_BoneSpace
    translationMode=BMM_Additive
    translationSpace=BCS_BoneSpace
    alphaBoolBlend {blendOption=Linear}
    boneToModify {boneName=base_hair_01_r}
    rotation {yaw=-5}
    translation {y=-1, z=1}
   }
   n87 = Get bIsTargeting()
   n88 = "Transform (Modify) Bone - Bone: base_hair_01_r"(ComponentPose=@n86, bAlphaBoolEnabled=@n87, Translation="0,1,0", Rotation="-5,0,0", Scale="1,1,1") {
    alphaInputType=Bool
    rotationMode=BMM_Additive
    rotationSpace=BCS_BoneSpace
    translationMode=BMM_Additive
    translationSpace=BCS_BoneSpace
    alphaBoolBlend {blendOption=Cubic}
    boneToModify {boneName=base_hair_01_r}
    rotation {pitch=-5}
    translation {y=1}
   }
   n89 = "Transform (Modify) Bone - Bone: dyn_hair_back_01_r"(ComponentPose=@n88, bAlphaBoolEnabled=@n87, Translation="0,0,1", Rotation="2,0,0", Scale="1,1,1") {
    alphaInputType=Bool
    rotationMode=BMM_Additive
    rotationSpace=BCS_BoneSpace
    translationMode=BMM_Additive
    translationSpace=BCS_BoneSpace
    alphaBoolBlend {blendOption=Cubic}
    boneToModify {boneName=dyn_hair_back_01_r}
    rotation {pitch=2}
    translation {z=1}
   }
   n90 = Get head_rot_back_Extreme()
   n91 = "Transform (Modify) Bone - Bone: base_hair_back_02_c"!(ComponentPose=@n89, Alpha:float=@n90, Translation="3,2,0", Rotation="0,0,0", Scale="1,1,1") -> (Pose) {
    rotationMode=BMM_Additive
    rotationSpace=BCS_BoneSpace
    translationMode=BMM_Additive
    translationSpace=BCS_BoneSpace
    alphaBoolBlend {blendOption=Linear}
    boneToModify {boneName=base_hair_back_02_c}
    translation=<3 2>
   }
   n92 = "Transform (Modify) Bone - Bone: base_hair_back_01_r"(ComponentPose=@n91, Alpha=@n90, Translation="1,0,-2", Rotation="0,-5,0", Scale="1,1,1") {
    rotationMode=BMM_Additive
    rotationSpace=BCS_BoneSpace
    translationMode=BMM_Additive
    translationSpace=BCS_BoneSpace
    alphaBoolBlend {blendOption=Linear}
    boneToModify {boneName=base_hair_back_01_r}
    rotation {yaw=-5}
    translation {x=1, z=-2}
   }
   n93 = "Transform (Modify) Bone - Bone: cape"(ComponentPose=@n92, Alpha=Get head_rot_back(), Translation="2,-1.1,0", Rotation="0,12,0", Scale="1,1,1") {
    rotationMode=BMM_Additive
    rotationSpace=BCS_BoneSpace
    translationMode=BMM_Additive
    translationSpace=BCS_BoneSpace
    alphaBoolBlend {blendOption=Linear}
    boneToModify {boneName=cape}
    rotation {yaw=12}
    translation=<2 -1.1>
   }
   n94 = Get head_rot_back_neg()
   n95 = "Transform (Modify) Bone - Bone: base_hair_back_01_r"(ComponentPose=@n93, Alpha=@n94, Translation="0,0,0", Rotation="45,0,0", Scale="1,1,1") {
    rotationMode=BMM_Additive
    rotationSpace=BCS_BoneSpace
    translationMode=BMM_Additive
    translationSpace=BCS_BoneSpace
    alphaBoolBlend {blendOption=Linear}
    boneToModify {boneName=base_hair_back_01_r}
    rotation {pitch=45}
   }
   n96 = "Transform (Modify) Bone - Bone: base_hair_back_02_r"(ComponentPose=@n95, Alpha=@n94, Translation="0,0,0", Rotation="45,0,0", Scale="1,1,1") {
    rotationMode=BMM_Additive
    rotationSpace=BCS_BoneSpace
    translationMode=BMM_Additive
    translationSpace=BCS_BoneSpace
    alphaBoolBlend {blendOption=Linear}
    boneToModify {boneName=base_hair_back_02_r}
    rotation {pitch=45}
   }
   n97 = "Transform (Modify) Bone - Bone: base_hair_back_01_c"(ComponentPose=@n96, Alpha=@n94, Translation="0,0,0", Rotation="45,0,0", Scale="1,1,1") {
    rotationMode=BMM_Additive
    rotationSpace=BCS_BoneSpace
    translationMode=BMM_Additive
    translationSpace=BCS_BoneSpace
    alphaBoolBlend {blendOption=Linear}
    boneToModify {boneName=base_hair_back_01_c}
    rotation {pitch=45}
   }
   n98 = "Transform (Modify) Bone - Bone: base_hair_back_02_c"(ComponentPose=@n97, Alpha=@n94, Translation="0,0,0", Rotation="45,0,0", Scale="1,1,1") {
    rotationMode=BMM_Additive
    rotationSpace=BCS_BoneSpace
    translationMode=BMM_Additive
    translationSpace=BCS_BoneSpace
    alphaBoolBlend {blendOption=Linear}
    boneToModify {boneName=base_hair_back_02_c}
    rotation {pitch=45}
   }
   n99 = "Transform (Modify) Bone - Bone: base_hair_back_01_l"(ComponentPose=@n98, Alpha=@n94, Translation="0,0,0", Rotation="45,0,0", Scale="1,1,1") {
    rotationMode=BMM_Additive
    rotationSpace=BCS_BoneSpace
    translationMode=BMM_Additive
    translationSpace=BCS_BoneSpace
    alphaBoolBlend {blendOption=Linear}
    boneToModify {boneName=base_hair_back_01_l}
    rotation {pitch=45}
   }
   n100 = "Transform (Modify) Bone - Bone: base_hair_back_02_l"(ComponentPose=@n99, Alpha=@n94, Translation="0,0,0", Rotation="45,0,0", Scale="1,1,1") {
    rotationMode=BMM_Additive
    rotationSpace=BCS_BoneSpace
    translationMode=BMM_Additive
    translationSpace=BCS_BoneSpace
    alphaBoolBlend {blendOption=Linear}
    boneToModify {boneName=base_hair_back_02_l}
    rotation {pitch=45}
   }
   n101 = Control Rig!(Source=Component To Local(ComponentPose=@n100), Alpha, onlypistol:bool=Get Only_Pistol()) -> (Pose) {
    alphaBoolBlend {blendOption=Linear}
    controlRigAssetReference {blueprintRigClass=&F_MED_PartyJelly_FaceAcc_front_hair_CtrlRig_C}
   }
   n102 = RigidBody(ComponentPose=Local To Component(LocalPose=@n101), OverrideWorldGravity=Get GravityOverride()) {
    bOverrideWorldGravity=true
    overridePhysicsAsset=&F_MED_PartyJelly_FaceAcc_Physics
    simulationSpace=BaseBoneSpace
    alphaBoolBlend {blendOption=Cubic}
    baseBoneRef {boneName=head}
    overrideWorldGravity {z=-980}
    simSpaceSettings {maxAngularAcceleration=40, maxAngularVelocity=8, maxLinearAcceleration=2000, maxLinearVelocity=500, worldAlpha=1}
   }
   n103 = "Constraint - dyn_hair_back_01_c (PPP)"!(ComponentPose=@n102, Alpha=1, ConstraintWeights_0=0.5, ConstraintWeights_1:float=0.25, ConstraintWeights_2:float=0.25) -> (Pose) {
    alphaBoolBlend {blendOption=Cubic}
    alphaScaleBiasClamp {bInterpResult=true, interpSpeedIncreasing=5}
    boneToModify {boneName=dyn_hair_back_01_c}
    constraintSetup [
     {perAxis {bZ=false}, targetBone {boneName=spine_05}}
     {perAxis {bZ=false}, targetBone {boneName=dyn_hair_back_01_r}}
     {targetBone {boneName=dyn_hair_back_01_l}}
    ]
    constraintWeights [1,1,1]
   }
   n104 = "Constraint - dyn_hair_back_02_c (PPP)"!(ComponentPose=@n103, Alpha:float=1, ConstraintWeights_0=0.5, ConstraintWeights_1=0.25, ConstraintWeights_2=0.25) -> (Pose) {
    alphaBoolBlend {blendOption=Cubic}
    alphaScaleBiasClamp {bInterpResult=true, interpSpeedIncreasing=5}
    boneToModify {boneName=dyn_hair_back_02_c}
    constraintSetup [
     {perAxis {bZ=false}, targetBone {boneName=spine_05}}
     {perAxis {bZ=false}, targetBone {boneName=dyn_hair_back_02_r}}
     {targetBone {boneName=dyn_hair_back_02_l}}
    ]
    constraintWeights [1,1,1]
   }
   n105 = "Constraint - dyn_hair_back_03_c (PPP)"!(ComponentPose=@n104, Alpha:float=1, ConstraintWeights_0=0.5, ConstraintWeights_1=0.25, ConstraintWeights_2=0.25) -> (Pose) {
    alphaBoolBlend {blendOption=Cubic}
    alphaScaleBiasClamp {bInterpResult=true, interpSpeedIncreasing=5}
    boneToModify {boneName=dyn_hair_back_03_c}
    constraintSetup [
     {perAxis {bZ=false}, targetBone {boneName=spine_05}}
     {perAxis {bZ=false}, targetBone {boneName=base_hair_back_03_r}}
     {targetBone {boneName=base_hair_back_03_l}}
    ]
    constraintWeights [1,1,1]
   }
   n106 = Get wearing_backpack_Alpha() -> (wearing_backpack_Alpha:double)
   n107 = "Transform (Modify) Bone - Bone: base_hair_back_02_c"(ComponentPose=@n105, Alpha=@n106, Translation="0,0,0", Rotation="-60,0,0", Scale="1,1,1") {
    rotationMode=BMM_Additive
    rotationSpace=BCS_BoneSpace
    translationMode=BMM_Additive
    translationSpace=BCS_BoneSpace
    alphaBoolBlend {blendOption=Linear}
    alphaScaleBiasClamp {bInterpResult=true, interpSpeedDecreasing=4, interpSpeedIncreasing=4}
    boneToModify {boneName=base_hair_back_02_c}
    rotation {pitch=-60}
   }
   n108 = "Transform (Modify) Bone - Bone: base_hair_back_03_c"!(ComponentPose=@n107, Alpha:float=@n106, Translation="0,0,0", Rotation="-10,0,0", Scale="1,1,1") -> (Pose) {
    rotationMode=BMM_Additive
    rotationSpace=BCS_BoneSpace
    translationMode=BMM_Additive
    translationSpace=BCS_BoneSpace
    alphaBoolBlend {blendOption=Linear}
    alphaScaleBiasClamp {bInterpResult=true, interpSpeedDecreasing=4, interpSpeedIncreasing=4}
    boneToModify {boneName=base_hair_back_03_c}
    rotation {pitch=-10}
   }
   n109 = "Constraint - dyn_hair_01_r (R)"(ComponentPose=@n108, Alpha=0.6, ConstraintWeights_0=1) {
    alphaBoolBlend {blendOption=Linear}
    boneToModify {boneName=dyn_hair_01_r}
    constraintSetup [{transformType=Rotation, targetBone {boneName=base_hair_back_01_r}}]
    constraintWeights [1]
   }
   n110 = "Transform (Modify) Bone - Bone: base_hair_01_r"!(ComponentPose=@n109, Alpha:float=Get Head Left(), Translation="0,-2,0", Rotation="0,0,0", Scale="1,1,1") -> (Pose) {
    rotationMode=BMM_Additive
    rotationSpace=BCS_BoneSpace
    translationMode=BMM_Additive
    translationSpace=BCS_BoneSpace
    alphaBoolBlend {blendOption=Linear}
    alphaScaleBiasClamp {bInterpResult=true, interpSpeedDecreasing=4, interpSpeedIncreasing=4}
    boneToModify {boneName=base_hair_01_r}
    translation {y=-2}
   }
   n111 = "Transform (Modify) Bone - Bone: base_hair_01_r"(ComponentPose=@n110, Alpha=Get Head Right(), Translation="0,-2,0", Rotation="0,0,0", Scale="1,1,1") {
    rotationMode=BMM_Additive
    rotationSpace=BCS_BoneSpace
    translationMode=BMM_Additive
    translationSpace=BCS_BoneSpace
    alphaBoolBlend {blendOption=Linear}
    alphaScaleBiasClamp {bInterpResult=true, interpSpeedDecreasing=4, interpSpeedIncreasing=4}
    boneToModify {boneName=base_hair_01_r}
    translation {y=-2}
   }
   n112 = Control Rig!(
    Source=Component To Local(ComponentPose=@n111)
    Alpha
    Only_Rifle:bool=Get Only_Rifle()
    Pistol_Crouch:bool=Get Pistol_Crouch()
    wearing_backpack:bool=Get wearing_backpack_jump()
   ) -> (Pose) {
    alphaBoolBlend {blendOption=Linear}
    controlRigAssetReference {blueprintRigClass=&F_MED_PartyJelly_FaceAcc_front_Blob_CtrlRig_C}
   }
   n113 = "Transform (Modify) Bone - Bone: base_hair_01_r"(ComponentPose=Local To Component(LocalPose=@n112), Alpha=Get crouch_Walking(), Translation="0,2,0", Rotation="0,10,0", Scale="1,1,1") {
    rotationMode=BMM_Additive
    rotationSpace=BCS_BoneSpace
    translationMode=BMM_Additive
    translationSpace=BCS_BoneSpace
    alphaBoolBlend {blendOption=Linear}
    alphaScaleBiasClamp {bInterpResult=true}
    boneToModify {boneName=base_hair_01_r}
    rotation {yaw=10}
    translation {y=2}
   }
   n114 = "Transform (Modify) Bone - Bone: base_hair_01_r"(ComponentPose=@n113, Alpha=Get Head Down(), Translation="0,1,0", Rotation="0,-15,0", Scale="1,1,1") {
    rotationMode=BMM_Additive
    rotationSpace=BCS_BoneSpace
    translationMode=BMM_Additive
    translationSpace=BCS_BoneSpace
    alphaBoolBlend {blendOption=Linear}
    alphaScaleBiasClamp {bInterpResult=true}
    boneToModify {boneName=base_hair_01_r}
    rotation {yaw=-15}
    translation {y=1}
   }
   n115 = "Transform (Modify) Bone - Bone: dyn_hair_02_l"!(ComponentPose=@n114, bAlphaBoolEnabled:bool=Get bIsSkydiveFloating(), Translation="0,0,0", Rotation="0,50,0", Scale="1,1,1") -> (Pose) {
    alphaInputType=Bool
    rotationMode=BMM_Additive
    rotationSpace=BCS_BoneSpace
    translationMode=BMM_Additive
    translationSpace=BCS_BoneSpace
    alphaBoolBlend {blendOption=Linear}
    boneToModify {boneName=dyn_hair_02_l}
    rotation {yaw=50}
   }
   n116 = Get GravityOverride()
   n117 = Anim Dynamics - Bone: dyn_hair_front_01_r(ComponentPose=@n115, Alpha=1, GravityOverride=@n116, ExternalForce="0,0,0") {
    angularDampingOverride=1
    angularSpringConstant=20000
    bAngularSpring=true
    bOverrideAngularDamping=true
    bUseGravityOverride=true
    bUseSphericalLimits=true
    gravityScale=0.35
    simulationSpace=RootRelative
    alphaBoolBlend {blendOption=Linear}
    boundBone {boneName=dyn_hair_front_01_r}
    componentLinearAccScale=<1 1 1>
    componentLinearVelScale=<1 1 1>
    gravityOverride {z=-980}
    physicsBodyDefinitions [{
      boundBone {boneName=dyn_hair_front_01_r}
      boxExtents=<20 30 30>
      constraintSetup {angularLimitsMax=<2 2 2>, angularLimitsMin=<-2 -2 -2>, angularTarget=<1 0 0>}
      localJointOffset=<4 0 0>
     }]
   }
   n118 = Anim Dynamics - Bone: dyn_hair_front_02_r(ComponentPose=@n117, Alpha=1, GravityOverride=@n116, ExternalForce="0,0,0") {
    angularDampingOverride=1
    angularSpringConstant=10000
    bAngularSpring=true
    bOverrideAngularDamping=true
    bUseGravityOverride=true
    bUseSphericalLimits=true
    gravityScale=0.35
    simulationSpace=RootRelative
    alphaBoolBlend {blendOption=Linear}
    boundBone {boneName=dyn_hair_front_02_r}
    componentLinearAccScale=<1 1 1>
    componentLinearVelScale=<1 1 1>
    gravityOverride {z=-980}
    physicsBodyDefinitions [{
      boundBone {boneName=dyn_hair_front_02_r}
      boxExtents=<20 30 30>
      constraintSetup {angularLimitsMax=<10 10 10>, angularLimitsMin=<-10 -10 -10>, angularTarget=<1 0 0>}
      localJointOffset=<4 0 0>
     }]
    sphericalLimits [{limitRadius=8, drivingBone {boneName=upperarm_r}, sphereLocalOffset=<0 0 0>}]
   }
   n119 = Anim Dynamics - Bone: dyn_hair_front_03_r(ComponentPose=@n118, Alpha=1, GravityOverride=@n116, ExternalForce="0,0,0") {
    angularDampingOverride=1
    angularSpringConstant=10000
    bAngularSpring=true
    bOverrideAngularDamping=true
    bUseGravityOverride=true
    bUsePlanarLimit=false
    bUseSphericalLimits=true
    gravityScale=0.35
    simulationSpace=RootRelative
    alphaBoolBlend {blendOption=Linear}
    boundBone {boneName=dyn_hair_front_03_r}
    componentLinearAccScale=<1 1 1>
    componentLinearVelScale=<1 1 1>
    gravityOverride {z=-980}
    physicsBodyDefinitions [{
      boundBone {boneName=dyn_hair_front_03_r}
      boxExtents=<20 30 30>
      constraintSetup {angularLimitsMax=<15 15 15>, angularLimitsMin=<-15 -15 -15>, angularTarget=<1 0 0>}
      localJointOffset=<4 0 0>
     }]
    sphericalLimits [
     {limitRadius=10, drivingBone {boneName=clavicle_r}, sphereLocalOffset=<-5.2979 2.5497 2.95>}
     {limitRadius=8, drivingBone {boneName=neck_02}, sphereLocalOffset=<-3.3594 4.5904 0.4087>}
     {limitRadius=9, drivingBone {boneName=upperarm_r}, sphereLocalOffset=<-11.7598 0.1042 -0.7702>}
     {limitRadius=0, drivingBone {boneName=None}, sphereLocalOffset=<0 0 0>}
     {limitRadius=0, drivingBone {boneName=None}, sphereLocalOffset=<0 0 0>}
    ]
   }
   n120 = Anim Dynamics - Bone: dyn_hair_front_01_l(ComponentPose=@n119, Alpha=1, GravityOverride=@n116, ExternalForce="0,0,0") {
    angularDampingOverride=1
    angularSpringConstant=10000
    bAngularSpring=true
    bOverrideAngularDamping=true
    bUseGravityOverride=true
    bUsePlanarLimit=false
    bUseSphericalLimits=true
    gravityScale=0.35
    simulationSpace=RootRelative
    alphaBoolBlend {blendOption=Linear}
    boundBone {boneName=dyn_hair_front_01_l}
    componentLinearAccScale=<1 1 1>
    componentLinearVelScale=<1 1 1>
    gravityOverride {z=-980}
    physicsBodyDefinitions [{
      boundBone {boneName=dyn_hair_front_01_l}
      boxExtents=<20 30 30>
      constraintSetup {angularLimitsMax=<5 5 5>, angularLimitsMin=<-5 -5 -5>, angularTarget=<1 0 0>}
      localJointOffset=<4 0 0>
     }]
    sphericalLimits [{limitRadius=0, drivingBone {boneName=None}, sphereLocalOffset=<0 0 0>}]
   }
   n121 = Anim Dynamics - Bone: dyn_hair_front_02_l(ComponentPose=@n120, Alpha=1, GravityOverride=@n116, ExternalForce="0,0,0") {
    angularDampingOverride=1
    angularSpringConstant=10000
    bAngularSpring=true
    bOverrideAngularDamping=true
    bUseGravityOverride=true
    bUsePlanarLimit=false
    bUseSphericalLimits=true
    gravityScale=0.35
    simulationSpace=RootRelative
    alphaBoolBlend {blendOption=Linear}
    boundBone {boneName=dyn_hair_front_02_l}
    componentLinearAccScale=<1 1 1>
    componentLinearVelScale=<1 1 1>
    gravityOverride {z=-980}
    physicsBodyDefinitions [{
      boundBone {boneName=dyn_hair_front_02_l}
      boxExtents=<20 30 30>
      constraintSetup {angularLimitsMax=<10 10 10>, angularLimitsMin=<-10 -10 -10>, angularTarget=<1 0 0>}
      localJointOffset=<4 0 0>
     }]
    sphericalLimits [
     {limitRadius=7, drivingBone {boneName=upperarm_l}, sphereLocalOffset=<0 0 0>}
     {limitRadius=0, drivingBone {boneName=None}, sphereLocalOffset=<0 0 0>}
    ]
   }
   n122 = Anim Dynamics - Bone: dyn_hair_front_03_l(ComponentPose=@n121, Alpha=1, GravityOverride=@n116, ExternalForce="0,0,0") {
    angularDampingOverride=1
    angularSpringConstant=10000
    bAngularSpring=true
    bOverrideAngularDamping=true
    bUseGravityOverride=true
    bUsePlanarLimit=false
    gravityScale=0.35
    simulationSpace=RootRelative
    alphaBoolBlend {blendOption=Linear}
    boundBone {boneName=dyn_hair_front_03_l}
    componentLinearAccScale=<1 1 1>
    componentLinearVelScale=<1 1 1>
    gravityOverride {z=-980}
    physicsBodyDefinitions [{
      boundBone {boneName=dyn_hair_front_03_l}
      boxExtents=<20 30 30>
      constraintSetup {angularLimitsMax=<10 10 10>, angularLimitsMin=<-10 -10 -10>, angularTarget=<1 0 0>}
      localJointOffset=<0 0 0>
     }]
    planarLimits [{drivingBone {boneName=None}, planeTransform {location=<0 0 0>, rotation=<rot 0 0 0>}}]
    sphericalLimits [
     {limitRadius=9, drivingBone {boneName=clavicle_l}, sphereLocalOffset=<7.3163 0 -0.2805>}
     {limitRadius=9, drivingBone {boneName=neck_02}, sphereLocalOffset=<-5.6783 4.5904 -1.0203>}
     {limitRadius=0, drivingBone {boneName=None}, sphereLocalOffset=<0 0 0>}
     {limitRadius=0, drivingBone {boneName=None}, sphereLocalOffset=<0 0 0>}
    ]
   }
   n123 = "Transform (Modify) Bone - Bone: base_hair_02_l"!(ComponentPose=@n122, Alpha:float=1, Translation="0,0,0", Rotation="-15,0,0", Scale="1,1,1") -> (Pose) {
    rotationMode=BMM_Additive
    rotationSpace=BCS_BoneSpace
    translationMode=BMM_Additive
    translationSpace=BCS_BoneSpace
    alphaBoolBlend {blendOption=Cubic}
    alphaScaleBiasClamp {bInterpResult=true, interpSpeedIncreasing=2}
    boneToModify {boneName=base_hair_02_l}
    rotation {pitch=-15}
   }
   n124 = "Transform (Modify) Bone - Bone: base_hair_03_l"!(ComponentPose=@n123, Alpha:float=1, Translation="0,0,0", Rotation="-5,0,0", Scale="1,1,1") -> (Pose) {
    rotationMode=BMM_Additive
    rotationSpace=BCS_BoneSpace
    translationMode=BMM_Additive
    translationSpace=BCS_BoneSpace
    alphaBoolBlend {blendOption=Cubic}
    alphaScaleBiasClamp {bInterpResult=true, interpSpeedIncreasing=2}
    boneToModify {boneName=base_hair_03_l}
    rotation {pitch=-5}
   }
   n125 = "Transform (Modify) Bone - Bone: base_hair_03_r"!(ComponentPose=@n124, Alpha:float=1, Translation="0,0,0", Rotation="-20,0,0", Scale="1,1,1") -> (Pose) {
    rotationMode=BMM_Additive
    rotationSpace=BCS_BoneSpace
    translationMode=BMM_Additive
    translationSpace=BCS_BoneSpace
    alphaBoolBlend {blendOption=Linear}
    alphaScaleBiasClamp {bInterpResult=true, interpSpeedIncreasing=2}
    boneToModify {boneName=base_hair_03_r}
    rotation {pitch=-20}
   }
   n126 = "Transform (Modify) Bone - Bone: base_hair_01_l"(ComponentPose=@n125, Alpha=1, Translation="0,0,0", Rotation="8,12,0", Scale="1,1,1") {
    rotationMode=BMM_Additive
    rotationSpace=BCS_BoneSpace
    translationMode=BMM_Additive
    translationSpace=BCS_BoneSpace
    alphaBoolBlend {blendOption=Linear}
    alphaScaleBiasClamp {bInterpResult=true, interpSpeedIncreasing=2}
    boneToModify {boneName=base_hair_01_l}
    rotation {pitch=8, yaw=12}
   }
   n127 = "Transform (Modify) Bone - Bone: base_hair_01_r"(ComponentPose=@n126, Alpha=1, Translation="0,0,0", Rotation="10,5,0", Scale="1,1,1") {
    rotationMode=BMM_Additive
    rotationSpace=BCS_BoneSpace
    translationMode=BMM_Additive
    translationSpace=BCS_BoneSpace
    alphaBoolBlend {blendOption=Linear}
    alphaScaleBiasClamp {bInterpResult=true, interpSpeedIncreasing=2}
    boneToModify {boneName=base_hair_01_r}
    rotation {pitch=10, yaw=5}
   }
   n128 = "Transform (Modify) Bone - Bone: base_hair_01_r"(ComponentPose=@n127, Alpha=1, Translation="-1,0,0", Rotation="-2,0,0", Scale="1,1,1") {
    rotationMode=BMM_Additive
    rotationSpace=BCS_BoneSpace
    translationMode=BMM_Additive
    translationSpace=BCS_BoneSpace
    alphaBoolBlend {blendOption=Linear}
    alphaScaleBiasClamp {bInterpResult=true, interpSpeedDecreasing=20, interpSpeedIncreasing=4}
    boneToModify {boneName=base_hair_01_r}
    rotation {pitch=-2}
    translation {x=-1}
   }
   n129 = RigidBody(ComponentPose=@n128, OverrideWorldGravity=Get GravityOverride()) {
    bOverrideWorldGravity=true
    overridePhysicsAsset=&/F_MED_PartyJelly_FaceAcc_FrontEnd_Emote_Physics
    simulationSpace=BaseBoneSpace
    alphaBoolBlend {blendOption=Cubic}
    alphaScaleBiasClamp {bInterpResult=true, interpSpeedDecreasing=20, interpSpeedIncreasing=4}
    baseBoneRef {boneName=head}
    overrideWorldGravity {z=-980}
    simSpaceSettings {maxAngularAcceleration=40, maxAngularVelocity=8, maxLinearAcceleration=2000, maxLinearVelocity=500, worldAlpha=1}
   }
   n130 = "Transform (Modify) Bone - Bone: dyn_hair_02_r"!(ComponentPose=@n129, Alpha:float=1, Translation="0,0,0", Rotation="-5,0,0", Scale="1,1,1") -> (Pose) {
    rotationMode=BMM_Additive
    rotationSpace=BCS_BoneSpace
    translationMode=BMM_Additive
    translationSpace=BCS_BoneSpace
    alphaBoolBlend {blendOption=Linear}
    alphaScaleBiasClamp {bInterpResult=true, interpSpeedDecreasing=20, interpSpeedIncreasing=4}
    boneToModify {boneName=dyn_hair_02_r}
    rotation {pitch=-5}
   }
   n131 = "Transform (Modify) Bone - Bone: dyn_hair_03_r"!(ComponentPose=@n130, Alpha:float=1, Translation="0,0,0", Rotation="-10,0,0", Scale="1,1,1") -> (Pose) {
    rotationMode=BMM_Additive
    rotationSpace=BCS_BoneSpace
    translationMode=BMM_Additive
    translationSpace=BCS_BoneSpace
    alphaBoolBlend {blendOption=Linear}
    alphaScaleBiasClamp {bInterpResult=true, interpSpeedDecreasing=20, interpSpeedIncreasing=4}
    boneToModify {boneName=dyn_hair_03_r}
    rotation {pitch=-10}
   }
   n132 = "Transform (Modify) Bone - Bone: dyn_hair_02_l"!(ComponentPose=@n131, Alpha:float=1, Translation="0,0,0", Rotation="-5,0,0", Scale="1,1,1") -> (Pose) {
    rotationMode=BMM_Additive
    rotationSpace=BCS_BoneSpace
    translationMode=BMM_Additive
    translationSpace=BCS_BoneSpace
    alphaBoolBlend {blendOption=Linear}
    alphaScaleBiasClamp {bInterpResult=true, interpSpeedDecreasing=20, interpSpeedIncreasing=4}
    boneToModify {boneName=dyn_hair_02_l}
    rotation {pitch=-5}
   }
   n133 = "Transform (Modify) Bone - Bone: dyn_hair_03_l"(ComponentPose=@n132, Alpha=1, Translation="0,0,0", Rotation="-10,0,0", Scale="1,1,1") {
    rotationMode=BMM_Additive
    rotationSpace=BCS_BoneSpace
    translationMode=BMM_Additive
    translationSpace=BCS_BoneSpace
    alphaBoolBlend {blendOption=Linear}
    alphaScaleBiasClamp {bInterpResult=true, interpSpeedDecreasing=20, interpSpeedIncreasing=4}
    boneToModify {boneName=dyn_hair_03_l}
    rotation {pitch=-10}
   }
   n134 = "Transform (Modify) Bone - Bone: dyn_hair_01_r"!(ComponentPose=@n133, Alpha:float=1, Translation="0,0,0", Rotation="-15,-10,0", Scale="1,1,1") -> (Pose) {
    rotationMode=BMM_Additive
    rotationSpace=BCS_BoneSpace
    translationMode=BMM_Additive
    translationSpace=BCS_BoneSpace
    alphaBoolBlend {blendInTime=0.7, blendOption=Linear, blendOutTime=0.1}
    alphaScaleBiasClamp {bInterpResult=true, interpSpeedDecreasing=20, interpSpeedIncreasing=4}
    boneToModify {boneName=dyn_hair_01_r}
    rotation {pitch=-15, yaw=-10}
   }
   n135 = "Transform (Modify) Bone - Bone: dyn_hair_02_r"(ComponentPose=@n134, Alpha=1, Translation="0,0,0", Rotation="5,0,0", Scale="1,1,1") {
    rotationMode=BMM_Additive
    rotationSpace=BCS_BoneSpace
    translationMode=BMM_Additive
    translationSpace=BCS_BoneSpace
    alphaBoolBlend {blendInTime=0.7, blendOption=Linear, blendOutTime=0.1}
    alphaScaleBiasClamp {bInterpResult=true, interpSpeedDecreasing=20, interpSpeedIncreasing=4}
    boneToModify {boneName=dyn_hair_02_r}
    rotation {pitch=5}
   }
   n136 = "Transform (Modify) Bone - Bone: dyn_hair_03_r"(ComponentPose=@n135, Alpha=1, Translation="0,0,0", Rotation="5,0,0", Scale="1,1,1") {
    rotationMode=BMM_Additive
    rotationSpace=BCS_BoneSpace
    translationMode=BMM_Additive
    translationSpace=BCS_BoneSpace
    alphaBoolBlend {blendInTime=0.7, blendOption=Linear, blendOutTime=0.1}
    alphaScaleBiasClamp {bInterpResult=true, interpSpeedDecreasing=20, interpSpeedIncreasing=4}
    boneToModify {boneName=dyn_hair_03_r}
    rotation {pitch=5}
   }
   n137 = "Transform (Modify) Bone - Bone: dyn_hair_01_l"(ComponentPose=@n136, Alpha=1, Translation="0,0,0", Rotation="5,-10,0", Scale="1,1,1") {
    rotationMode=BMM_Additive
    rotationSpace=BCS_BoneSpace
    translationMode=BMM_Additive
    translationSpace=BCS_BoneSpace
    alphaBoolBlend {blendInTime=0.7, blendOption=Linear, blendOutTime=0.7}
    alphaScaleBiasClamp {bInterpResult=true, interpSpeedDecreasing=20, interpSpeedIncreasing=4}
    boneToModify {boneName=dyn_hair_01_l}
    rotation {pitch=5, yaw=-10}
   }
   n138 = "Transform (Modify) Bone - Bone: dyn_hair_02_l"(ComponentPose=@n137, Alpha=1, Translation="0,0,0", Rotation="3,0,0", Scale="1,1,1") {
    rotationMode=BMM_Additive
    rotationSpace=BCS_BoneSpace
    translationMode=BMM_Additive
    translationSpace=BCS_BoneSpace
    alphaBoolBlend {blendInTime=0.7, blendOption=Linear, blendOutTime=0.7}
    alphaScaleBiasClamp {bInterpResult=true, interpSpeedDecreasing=20, interpSpeedIncreasing=4}
    boneToModify {boneName=dyn_hair_02_l}
    rotation {pitch=3}
   }
   n139 = RigidBody(ComponentPose=@n138, OverrideWorldGravity=Get GravityOverride()) {
    bOverrideWorldGravity=true
    overridePhysicsAsset=&/F_MED_PartyJelly_FaceAcc_FrontEnd_Idle_Physics
    simulationSpace=BaseBoneSpace
    alphaBoolBlend {blendOption=Cubic}
    alphaScaleBiasClamp {bInterpResult=true, interpSpeedDecreasing=20, interpSpeedIncreasing=4}
    baseBoneRef {boneName=head}
    overrideWorldGravity {z=-980}
    simSpaceSettings {maxAngularAcceleration=40, maxAngularVelocity=8, maxLinearAcceleration=2000, maxLinearVelocity=500, worldAlpha=1}
   }
   Output Pose(Result=Component To Local(ComponentPose=@n139)) {layerGroup=None}
   n140 = Layered blend per bone(
    BasePose=Copy Pose From Mesh() {bUseAttachedParent=true}
    BlendPoses_0=Copy Pose From Mesh(SourceMeshComponent=Get Body() -> (Body:Skeletal Mesh Component Budgeted Object Reference)) {bUseAttachedParent=true}
   ) {layerSetup [{branchFilters [
       {blendDepth=2, boneName=clavicle_r}
       {blendDepth=2, boneName=clavicle_l}
       {blendDepth=2, boneName=upperarm_l}
       {blendDepth=2, boneName=upperarm_r}
      ]}]}
   Save cached pose 'InitialPose'(Pose=@n140) {cacheName=InitialPose}
   Get FrontEnd_not_emote()
   Get isWearingBackbling()
   Get wearing_backpack()
   Get EYouboreme_Alpha() -> (EYouboreme_Alpha:double)
  ]}
 EventGraph {nodes [
   Event BlueprintUpdateAnimation()
   n1 = BreakTransform(InTransform=GetDeltaTransformFromRefPose(self=GetOwnerSkeletalMeshComponent(), BoneName=head, BaseName=spine_05))
   > Set head_rot_back(head_rot_back:double=MapRangeClamped(Value=BreakRotator(InRot=@n1.Rotation).Yaw, InRangeA=0, InRangeB=-45)) -> (Output_Get:double)
   n2 = GetOwnerSkeletalMeshComponent()
   n3 = BreakRotator(InRot=BreakTransform(InTransform=GetDeltaTransformFromRefPose(self=@n2, BoneName=head, BaseName=spine_05)).Rotation)
   > n4 = Set head_rot_back_Extreme(head_rot_back_Extreme:double=MapRangeClamped(Value=@n3.Yaw, InRangeA=0, InRangeB=-75)) -> (Output_Get:double)
   n5 = Get bIsCrouchSprinting()
   > n6 = Set head_rot_back_Extreme_Crouch_Bool(head_rot_back_Extreme_Crouch_Bool:bool=OR Boolean!(A=Get bIsCrouching(), B=@n5, C:bool) -> (ReturnValue)) -> (Output_Get:bool)
   > n7 = Set head_rot_back_Extreme_Crouch(head_rot_back_Extreme_Crouch:double="To Float (Boolean)"(InBool=@n6)) -> (Output_Get:double)
   > Set iscouching_float(iscouching_float:double=Lerp(A=@n4, B=@n7, Alpha=0)) -> (Output_Get:double)
   > Set iscouching_Walking(iscouching_Walking:double=SelectFloat(A=1, B=0, bPickA=@n5)) -> (Output_Get:double)
   > n8 = "Gravity Override State Function (Parameter Asset) (Nativized)"(
    CharacterPartAnimInstance=Self-Reference()
    GameTimeInSeconds=GetGameTimeInSeconds()
    Parameters=&/M_MED_PartyJelly_FaceAcc_Parameters
   )
   > n9 = Set GravityOverride(GravityOverride=@n8)
   > Set GravityOverride(GravityOverride=@n9)
   n10 = BreakTransform(InTransform=GetDeltaTransformFromRefPose(self=GetOwnerSkeletalMeshComponent(), BoneName=head, BaseName=spine_05))
   > Set head_rot_back_neg(head_rot_back_neg:double=MapRangeClamped(Value=BreakRotator(InRot=@n10.Rotation).Yaw, InRangeA=0, InRangeB=75)) -> (Output_Get:double)
   > Set iscouching_targetting_Walking(iscouching_targetting_Walking:bool=OR Boolean(A=false, B=Get bIsCrouchSprinting(), C=Get bIsTargeting())) -> (Output_Get:bool)
   > Set head_rot_back_min(head_rot_back_min:double=MapRangeClamped(Value=@n3.Yaw, InRangeA=0, InRangeB=0, OutRangeA=-8)) -> (Output_Get:double)
   n11 = GetOwnerSkeletalMeshComponent()
   n12 = BreakRotator(InRot=BreakTransform(InTransform=GetDeltaTransformFromRefPose(self=@n11, BoneName=head, BaseName=spine_05)).Rotation)
   n13 = Get bIsTargeting()
   n14 = SelectFloat(A=MapRangeClamped(Value=@n12.Yaw, InRangeA=-3, InRangeB=-40), B=0, bPickA=@n13)
   > n15 = Set HeadUp(HeadUp:double=@n14) -> (Output_Get:double)
   n16 = Get bIsTargeting()
   n17 = SelectFloat(A=MapRangeClamped(Value=@n12.Yaw, InRangeA=1, InRangeB=10), B=0, bPickA=@n16)
   > Set Head Down(Head Down:double=@n17) -> (Output_Get:double)
   > Set Head Left(Head Left:double=MapRangeClamped(Value=@n12.Pitch, InRangeA=1, InRangeB=-1)) -> (Output_Get:double)
   > Set pistol_up(pistol_up:double=@n14) -> (Output_Get:double)
   > Set pistol_down(pistol_down:double=@n17) -> (Output_Get:double)
   n18 = Get bIsCrouching()
   n19 = Get bIsMoving2D()
   > Set crouch_Walking(crouch_Walking:double=SelectFloat(A=@n15, B=0, bPickA=AND Boolean!(A=@n18, B=@n19, C:bool=@n16) -> (ReturnValue))) -> (Output_Get:double)
   n20 = Get bIsMoving2D()
   n21 = NOT Boolean(A=@n20)
   n22 = Get bIsCrouching()
   n23 = Get bIsTargeting()
   > Set Crouch_target(Crouch_target:double=SelectFloat(A=@n15, B=0, bPickA=AND Boolean(A=@n22, B=@n23, C=@n21))) -> (Output_Get:double)
   > Set Crouch_up(Crouch_up:double=SelectFloat(A=@n15, B=0, bPickA=Get bIsCrouching())) -> (Output_Get:double)
   n24 = NOT Boolean(A=Get bIsCrouching())
   n25 = Get bIsTargeting()
   > Set Rifle_Up(Rifle_Up:double=SelectFloat(A=Get HeadUp() -> (HeadUp:double), B=0, bPickA=AND Boolean!(A=@n25, B=@n24) -> (ReturnValue))) -> (Output_Get:double)
   n26 = Get bIsCrouching()
   n27 = Get bIsTargeting()
   n28 = Get HeadUp()
   > Set Rifle_Crouch_Up(Rifle_Crouch_Up:double=SelectFloat(A=Get Rifle_Up(), B=@n28, bPickA=AND Boolean(A=@n26, B=@n27))) -> (Output_Get:double)
   n29 = Get bIsCrouching()
   n30 = NOT Boolean(A=@n29)
   n31 = Get bIsTargeting()
   > Set Rifle_Stand_Up(Rifle_Stand_Up:double=SelectFloat(A=@n28, B=0, bPickA=AND Boolean!(A=@n30, B=@n31, C:bool=@n20) -> (ReturnValue))) -> (Output_Get:double)
   n32 = NOT Boolean(A=@n29)
   n33 = NOT Boolean(A=Get bIsMoving2D())
   n34 = Get bIsTargeting()
   > Set Pistol_Stand_Up(Pistol_Stand_Up:double=SelectFloat(A=Get pistol_up() -> (pistol_up:double), B=@n28, bPickA=AND Boolean(A=@n34, B=@n33, C=@n32))) -> (Output_Get:double)
   > Set Head Right(Head Right:double=MapRangeClamped(Value=@n12.Pitch, InRangeA=-1, InRangeB=1)) -> (Output_Get:double)
   n35 = Get bIsTargeting()
   n36 = Get bIsCrouching()
   > Set HairRot(HairRot:double=SelectFloat(A=@n28, B=0, bPickA=AND Boolean!(A=@n36, B=@n35) -> (ReturnValue))) -> (Output_Get:double)
   n37 = NOT Boolean(A=Get bIsTargeting())
   n38 = Get bIsMoving2D()
   n39 = Get bIsAccelerating2D()
   > Set Hair_ADJ(Hair_ADJ:double=SelectFloat(A=0, B=0, bPickA=AND Boolean!(A=@n37, B=@n39, C:bool=@n38) -> (ReturnValue))) -> (Output_Get:double)
   n40 = Get bIsSkyDiving()
   n41 = Get bIsSkydiveFloating()
   n42 = Get bIsParachuteOpened()
   > Set Sky Dive(Sky Dive:double=SelectFloat(A=0, B=0, bPickA=AND Boolean(A=@n41, B=@n40, C=@n42))) -> (Output_Get:double)
   n43 = GetAnimInstance(self=GetOwnerSkeletalMeshComponent())
   > Is Valid(InputObject=@n43)
   > Set EYouboreme_Not(EYouboreme_Not:bool=NOT Boolean(A=Montage_IsPlaying(self=@n43, Montage=/BRCosmetics/Animation/Game/MainPlayer/Emotes/YouBoreMe/Emote_You_Bore_Me_CMF_M))) -> (Output_Get:bool)
   n44 = Get bIsPlayingEmote()
   n45 = Get bIsInFrontEnd()
   > n46 = Set FrontEnd_not_emote(FrontEnd_not_emote:bool=AND Boolean!(A=@n45, B=NOT Boolean(A=@n44)) -> (ReturnValue)) -> (Output_Get:bool)
   > Set FrontEnd Not Emote Alpha(FrontEnd Not Emote Alpha:double="To Float (Boolean)"(InBool=@n46)) -> (Output_Get:double)
   n47 = Get EYouboreme_Not() -> (EYouboreme_Not:bool)
   > Set FrontEnd Emote Alpha(FrontEnd Emote Alpha:double="To Float (Boolean)"(InBool=AND Boolean!(A=@n45, B=@n44, C:bool=@n47) -> (ReturnValue))) -> (Output_Get:double)
   n48 = Get Pistol_Crouch()
   n49 = Get bIsMoving2D()
   > Set Pistol_Crouch_Walk(Pistol_Crouch_Walk:double=SelectFloat(A=Get pistol() -> (pistol:double), B=0, bPickA=AND Boolean!(A=@n49, B=@n48) -> (ReturnValue))) -> (Output_Get:double)
   > Set rifle_on_backpack(rifle_on_backpack:double=SelectFloat(A=Get Rifle() -> (Rifle:double), B=Get head_rot_back_Extreme(), bPickA=Get isWearingBackbling())) -> (Output_Get:double)
   n50 = AND Boolean!(A=Get bIsJumping(), B=Get isWearingBackbling(), C:bool=Get bIsFalling()) -> (ReturnValue)
   > Set wearing_backpack_Alpha(wearing_backpack_Alpha:double="To Float (Boolean)"(InBool=@n50)) -> (Output_Get:double)
   n51 = GetAnimInstance(self=GetOwnerSkeletalMeshComponent())
   > Is Valid(InputObject=@n51)
   n52 = NOT Boolean(A=Get bIsOnGround() -> (bIsOnGround:bool))
   n53 = Get bIsJumping()
   > n54 = Set EYouboreme(EYouboreme:bool=AND Boolean!(A=Get isWearingBackbling(), B=OR Boolean!(A=@n53, B=@n52) -> (ReturnValue)) -> (ReturnValue)) -> (Output_Get:bool)
   > Set EYouboreme_Alpha(EYouboreme_Alpha:double="To Float (Boolean)"(InBool=@n54)) -> (Output_Get:double)
   n55 = GetAnimInstance(self=GetOwnerSkeletalMeshComponent())
   > Is Valid(InputObject=@n55)
   > n56 = Set EmoteAnnoyingPen(EmoteAnnoyingPen:bool=Montage_IsPlaying(self=@n55)) -> (Output_Get:bool)
   > Set EmoteAnnoyingPen_Alpha(EmoteAnnoyingPen_Alpha:double="To Float (Boolean)"(InBool=@n56)) -> (Output_Get:double)
   n57 = Event OnCharacterCustomizationLoaded()
   > n58 = Sequence!(execute) -> (then_0, then_1, then_2:Exec, then_3:Exec)
   > Set isWearingBackbling(isWearingBackbling=CheckForBackbling())
   Set Body(
    execute=@n58.then_1
    Body:Skeletal Mesh Component Budgeted Object Reference=Get Array Item with EFort Custom Part Type(Array=@n57.CustomMeshParts)
   ) -> (Output_Get:Skeletal Mesh Component Budgeted Object Reference)
   Branch(execute=@n58.then_2, Condition=Get bIsTargeting())
   > Set ADS_Down(ADS_Down:double=Get Head Down()) -> (Output_Get:double)
   Event OnWeaponChanged()
   > n59 = Cast To FortPlayerAnimInstance(Object=GetAnimInstance(self=GetOwnerSkeletalMeshComponent()))
   > Set pistol(pistol:double=0) -> (Output_Get:double)
   > Set Melee(Melee:double=0) -> (Output_Get:double)
   > Set Rifle(Rifle:double=0) -> (Output_Get:double)
   > n60 = Switch on EFortWeaponCoreAnimation!(execute, Selection:EFortWeaponCoreAnimation Enum=Get WeaponCoreAnim(self=@n59)) -> (
    Melee:Exec
    Pistol:Exec
    Shotgun:Exec
    PaperBlueprint:Exec
    Rifle:Exec
    MeleeOneHand:Exec
    MachinePistol:Exec
    RocketLauncher:Exec
    GrenadeLauncher:Exec
    GoingCommando:Exec
    AssaultRifle:Exec
    TacticalShotgun:Exec
    SniperRifle:Exec
    TrapPlacement:Exec
    ShoulderLauncher:Exec
    AbilityDecoTool:Exec
    Crossbow:Exec
    C4:Exec
    RemoteControl:Exec
    DualWield:Exec
    AR_BullPup:Exec
    AR_ForwardGrip:Exec
    MedPackPaddles:Exec
    SMG_P90:Exec
    AR_DrumGun:Exec
    Consumable_Small:Exec
    Consumable_Large:Exec
    Balloon:Exec
    MountedTurret:Exec
    CreativeTool:Exec
    ExplosiveBow:Exec
    AshtonIndigo:Exec
    AshtonChicago:Exec
    MeleeDualWield:Exec
    Unarmed:Exec
   )
   Set Rifle(
    execute=[
     @n60.AR_BullPup
     @n60.AR_ForwardGrip
     @n60.AssaultRifle
     @n60.Crossbow
     @n60.GrenadeLauncher
     @n60.Rifle
     @n60.RocketLauncher
     @n60.Shotgun
     @n60.SniperRifle
     @n60.TacticalShotgun
    ]
    Rifle=1
   )
   > Set Only_Rifle(Only_Rifle:bool=1) -> (Output_Get:bool)
   > Set Only_Pistol(Only_Pistol:bool=false) -> (Output_Get:bool)
   > Set Pistol_Crouch(Pistol_Crouch:bool=false) -> (Output_Get:bool)
   Set Melee(execute=@n60.Melee, Melee=1)
   > Set Only_Rifle(Only_Rifle=false)
   > Set Only_Pistol(Only_Pistol=false)
   > Set Pistol_Crouch(Pistol_Crouch=false)
   Set pistol(execute=@n60.Pistol, pistol=1)
   > Set Only_Rifle(Only_Rifle=false)
   > Set Only_Pistol(Only_Pistol=true)
   > Set Pistol_Crouch(Pistol_Crouch=true)
   Montage_IsPlaying(self=@n51, Montage=&Emote_You_Bore_Me_CMF_M)
   SelectVector(A="0,0,-3000", B=@n9, bPickA=Get isWearingBackbling())
   "To Float (Boolean)"(InBool=Set wearing_backpack(wearing_backpack:bool=@n50) -> (Output_Get:bool))
   MapRangeClamped(Value=0, InRangeA=-3, InRangeB=-40)
   n61 = Get bIsCrouching()
   n62 = Get bIsMoving2D()
   AND Boolean(A=@n13, B=NOT Boolean(A=AND Boolean(A=@n61, B=@n62)))
   Get bIsFalling()
   Get bIsInAir() -> (bIsInAir:bool)
   Get Rifle_on backpack() -> (Rifle_on backpack:bool)
   Get bIsCrouching()
   Get wearing_backpack()
   Get bIsCrouching()
   Get wearing_backpack()
   Set Rifle_on backpack(Rifle_on backpack:bool=false) -> (Output_Get:bool)
   Set wearing_backpack_jump(wearing_backpack_jump:bool=false) -> (Output_Get:bool)
   Set Head Left(Head Left=0)
  ]}
]
variables [
 ADS_Down {type=number}
 ADS_Up {type=number}
 Body {type=SkeletalMeshComponentBudgeted}
 Crouch_target {type=number}
 Crouch_up {type=number}
 crouch_Walking {type=number}
 EmoteAnnoyingPen {type=boolean}
 EmoteAnnoyingPen_Alpha {type=number}
 EYouboreme {type=boolean}
 EYouboreme_Alpha {type=number}
 EYouboreme_Not {type=boolean}
 FrontEnd Emote Alpha {type=number}
 FrontEnd Not Emote Alpha {type=number}
 FrontEnd_Emote {type=boolean}
 FrontEnd_not_emote {type=boolean}
 Gravity Override Params {type=GravityOverrideParamsStruct}
 GravityOverride {type=Vector}
 Hair_ADJ {type=number}
 HairRot {type=number}
 Head Down {type=number}
 Head Left {type=number}
 Head Right {type=number}
 head_rot_back {type=number}
 head_rot_back_Extreme {type=number}
 head_rot_back_Extreme_Crouch {type=number}
 head_rot_back_Extreme_Crouch_Bool {type=boolean}
 head_rot_back_min {type=number}
 head_rot_back_neg {type=number}
 HeadUp {type=number}
 iscouching_float {type=number}
 iscouching_targetting_Walking {type=boolean}
 iscouching_Walking {type=number}
 isWearingBackbling {type=boolean}
 Melee {type=number}
 Not_Pistol {type=number}
 Only_Pistol {type=boolean}
 Only_Rifle {type=boolean}
 pistol {type=number}
 Pistol_Crouch {type=boolean}
 Pistol_Crouch_Walk {type=number}
 pistol_down {type=number}
 Pistol_Stand_Up {type=number}
 pistol_up {type=number}
 RBAN {type=number}
 Rifle {type=number}
 Rifle_Bool {type=boolean}
 Rifle_Crouch_Up {type=number}
 Rifle_on backpack {type=boolean}
 rifle_on_backpack {type=number}
 Rifle_Stand_Up {type=number}
 Rifle_Still {type=boolean}
 Rifle_Up {type=number}
 Sky Dive {type=number}
 wearing_backpack {type=boolean}
 wearing_backpack_Alpha {type=number}
 wearing_backpack_jump {type=boolean}
]
