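# Physics asset for the PartyJelly female medium character's face accessory, defining bodies for dynamic hair strands and the upper torso.
# Seven three-segment capsule chains (side, front, back) hang from a cape sphere, which joins spine_05 alongside clavicle and upper-arm capsules with free angular motion.
# Hair joints lock twist and mostly swing on one axis, with limits widening toward the tips.
# Arm capsules are asymmetric, notably larger on the right.
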PhysicsAsset /BRCosmetics/Characters/Player/Female/Medium/Bodies/F_MED_PartyJelly/Meshes/Parts/F_MED_PartyJelly_FaceAcc_Physics_A.F_MED_PartyJelly_FaceAcc_Physics
bodies {
 cape [{shapeType=Sphere, radius=5.625, shapeName=cape_sphere, center=<-8.3847 1.4523 0.0695>}]
 clavicle_l [{
   shapeType=Capsule
   length=22.5
   radius=7.005
   shapeName=clavicle_l_capsule
   center=<6.301 0.1459 -2.8271>
   rotation=<rot -0.0015 -84.9999 89.9999>
  }]
 clavicle_r [{
   shapeType=Capsule
   length=27.875
   radius=7.005
   shapeName=clavicle_r_capsule
   center=<-9.5856 -3.5617 0.1798>
   rotation=<rot -0.0017 -79.9999 -85.0002>
  }]
 dyn_hair_01_l [{shapeType=Capsule, length=3.375, radius=1.625, shapeName=dyn_hair_01_l_capsule, rotation=<rot 90 0 0>}]
 dyn_hair_01_r [{shapeType=Capsule, length=3.375, radius=1.625, shapeName=dyn_hair_01_r_capsule, rotation=<rot 90 0 0>}]
 dyn_hair_02_l [{shapeType=Capsule, length=3.125, radius=1.875, shapeName=dyn_hair_02_l_capsule, center=<0 0 0.3>, rotation=<rot 90 0 0>}]
 dyn_hair_02_r [{
   shapeType=Capsule
   length=2.125
   radius=2.5
   shapeName=dyn_hair_02_r_capsule
   center=<0.9377 -0.1474 1.0415>
   rotation=<rot 80.0001 179.9991 169.9996>
  }]
 dyn_hair_03_l [{
   shapeType=Capsule
   length=6.875
   radius=3.625
   shapeName=dyn_hair_03_l_capsule
   center=<3.603 0 1.2746>
   rotation=<rot 78.0128 -179.9993 -179.9994>
  }]
 dyn_hair_03_r [{
   shapeType=Capsule
   length=6.25
   radius=3.625
   shapeName=dyn_hair_03_r_capsule
   center=<4.4929 1.4507 1.3206>
   rotation=<rot 83.892 175.8224 160.905>
  }]
 dyn_hair_back_01_c [{shapeType=Capsule, length=3.375, radius=1.625, shapeName=dyn_hair_back_01_c_capsule, rotation=<rot 90 0 0>}]
 dyn_hair_back_01_l [{shapeType=Capsule, length=3.375, radius=1.625, shapeName=dyn_hair_back_01_l_capsule, rotation=<rot 90 0 0>}]
 dyn_hair_back_01_r [{shapeType=Capsule, length=3.375, radius=1.625, shapeName=dyn_hair_back_01_r_capsule, rotation=<rot 90 0 0>}]
 dyn_hair_back_02_c [{shapeType=Capsule, length=2.875, radius=2.125, shapeName=dyn_hair_back_02_c_capsule, center=<0 0 0.6>, rotation=<rot 79.7253 0 0>}]
 dyn_hair_back_02_l [{shapeType=Capsule, length=3.125, radius=1.875, shapeName=dyn_hair_back_02_l_capsule, center=<0 0 0.5>, rotation=<rot 90 0 0>}]
 dyn_hair_back_02_r [{
   shapeType=Capsule
   length=2.5
   radius=2.5
   shapeName=dyn_hair_back_02_r_capsule
   center=<0.0055 0 0.2966>
   rotation=<rot 79.9997 0.0005 0.0005>
  }]
 dyn_hair_back_03_c [{
   shapeType=Capsule
   length=2.25
   radius=3.75
   shapeName=dyn_hair_back_03_c_capsule
   center=<1.6546 0 0.11>
   rotation=<rot 69.6632 0.0001 0.0001>
  }]
 dyn_hair_back_03_l [{
   shapeType=Capsule
   length=6.625
   radius=2.875
   shapeName=dyn_hair_back_03_l_capsule
   center=<3.6001 0 0.0118>
   rotation=<rot 80.1954 -0.0006 -0.0007>
  }]
 dyn_hair_back_03_r [{
   shapeType=Capsule
   length=5.25
   radius=3.25
   shapeName=dyn_hair_back_03_r_capsule
   center=<3.0921 0 0.2843>
   rotation=<rot 88.496 0.0239 0.024>
  }]
 dyn_hair_front_01_l [{shapeType=Capsule, length=2.5, radius=2, shapeName=dyn_hair_front_01_l_capsule, rotation=<rot 90 0 0>}]
 dyn_hair_front_01_r [{shapeType=Capsule, length=2.5, radius=2, shapeName=dyn_hair_front_01_r_capsule, rotation=<rot 90 0 0>}]
 dyn_hair_front_02_l [{shapeType=Capsule, length=2.75, radius=1.75, shapeName=dyn_hair_front_02_l_capsule, rotation=<rot 90 0 0>}]
 dyn_hair_front_02_r [{shapeType=Capsule, length=7, radius=3, shapeName=dyn_hair_front_02_r_capsule, center=<1.4057 -0.733 0>, rotation=<rot 90 9.9994 0>}]
 dyn_hair_front_03_l [{shapeType=Capsule, length=5.5, radius=3.5, shapeName=dyn_hair_front_03_l_capsule, center=<1.7295 0 0.11>, rotation=<rot 90 180 178.2097>}]
 dyn_hair_front_03_r [{shapeType=Capsule, length=11, radius=3.625, shapeName=dyn_hair_front_03_r_capsule, center=<4.0125 -0.1825 -0.11>, rotation=<rot 90 9.9992 0>}]
 spine_05 [{shapeType=Capsule, length=22.505, radius=12.63, shapeName=spine_05_capsule, center=<3.012 5.4032 0>, rotation=<rot 0 -175.002 -180>}]
 upperarm_l [
  {shapeType=Capsule, length=0.505, radius=0.505, shapeName=upperarm_l_capsule, rotation=<rot 0 -90 0>}
  {
   shapeType=Capsule
   length=21.38
   radius=4.38
   shapeName=upperarm_l_capsule
   center=<18.8907 -6.575 0.1161>
   rotation=<rot 0.8667 -90.0758 -90.0384>
  }
 ]
 upperarm_r [
  {shapeType=Capsule, length=0.505, radius=1, shapeName=upperarm_r_capsule, rotation=<rot 0 -90 180>}
  {
   shapeType=Capsule
   length=41.755
   radius=8.875
   shapeName=upperarm_r_capsule
   center=<-11.9643 5.3626 1.1577>
   rotation=<rot 0.8646 -92.072 89.9314>
  }
 ]
}
constraints [
 {bone1Name=dyn_hair_02_l, bone2Name=dyn_hair_01_l, swing1LimitDegrees=0, swing2LimitDegrees=40, twistLimitDegrees=0}
 {bone1Name=dyn_hair_03_l, bone2Name=dyn_hair_02_l, swing1LimitDegrees=0, swing2LimitDegrees=50, twistLimitDegrees=0}
 {bone1Name=dyn_hair_02_r, bone2Name=dyn_hair_01_r, swing1LimitDegrees=0, swing2LimitDegrees=35, twistLimitDegrees=0}
 {bone1Name=dyn_hair_03_r, bone2Name=dyn_hair_02_r, swing1LimitDegrees=0, swing2LimitDegrees=45, twistLimitDegrees=0}
 {bone1Name=dyn_hair_back_02_c, bone2Name=dyn_hair_back_01_c, swing1LimitDegrees=0, swing2LimitDegrees=40, twistLimitDegrees=0}
 {bone1Name=dyn_hair_back_03_c, bone2Name=dyn_hair_back_02_c, swing1LimitDegrees=0, swing2LimitDegrees=50, twistLimitDegrees=0}
 {bone1Name=dyn_hair_back_02_l, bone2Name=dyn_hair_back_01_l, swing1LimitDegrees=0, swing2LimitDegrees=30, twistLimitDegrees=0}
 {bone1Name=dyn_hair_back_03_l, bone2Name=dyn_hair_back_02_l, swing1LimitDegrees=0, swing2LimitDegrees=40, twistLimitDegrees=0}
 {bone1Name=dyn_hair_back_02_r, bone2Name=dyn_hair_back_01_r, swing1LimitDegrees=0, swing2LimitDegrees=25, twistLimitDegrees=0}
 {bone1Name=dyn_hair_back_03_r, bone2Name=dyn_hair_back_02_r, swing1LimitDegrees=0, swing2LimitDegrees=35, twistLimitDegrees=0}
 {bone1Name=dyn_hair_front_02_l, bone2Name=dyn_hair_front_01_l, swing1LimitDegrees=20, swing2LimitDegrees=4, twistLimitDegrees=0}
 {bone1Name=dyn_hair_front_03_l, bone2Name=dyn_hair_front_02_l, swing1LimitDegrees=30, swing2LimitDegrees=6, twistLimitDegrees=0}
 {bone1Name=dyn_hair_front_02_r, bone2Name=dyn_hair_front_01_r, swing1LimitDegrees=20, swing2LimitDegrees=4, twistLimitDegrees=0}
 {bone1Name=dyn_hair_front_03_r, bone2Name=dyn_hair_front_02_r, swing1LimitDegrees=30, swing2LimitDegrees=6, twistLimitDegrees=0}
 {bone1Name=dyn_hair_back_01_r, bone2Name=cape, swing1LimitDegrees=0, swing2LimitDegrees=15, twistLimitDegrees=0}
 {bone1Name=dyn_hair_01_r, bone2Name=cape, swing1LimitDegrees=0, swing2LimitDegrees=20, twistLimitDegrees=0}
 {bone1Name=dyn_hair_front_01_r, bone2Name=cape, swing1LimitDegrees=10, swing2LimitDegrees=2, twistLimitDegrees=0}
 {bone1Name=dyn_hair_back_01_c, bone2Name=cape, swing1LimitDegrees=0, swing2LimitDegrees=20, twistLimitDegrees=0}
 {bone1Name=dyn_hair_back_01_l, bone2Name=cape, swing1LimitDegrees=0, swing2LimitDegrees=10, twistLimitDegrees=0}
 {bone1Name=dyn_hair_01_l, bone2Name=cape, swing1LimitDegrees=0, swing2LimitDegrees=20, twistLimitDegrees=0}
 {bone1Name=dyn_hair_front_01_l, bone2Name=cape, swing1LimitDegrees=10, swing2LimitDegrees=2, twistLimitDegrees=0}
 {
  bone1Name=clavicle_l
  bone2Name=spine_05
  swing1LimitDegrees=45
  swing1Motion=Free
  swing2LimitDegrees=45
  swing2Motion=Free
  twistLimitDegrees=45
  twistMotion=Free
 }
 {
  bone1Name=clavicle_r
  bone2Name=spine_05
  swing1LimitDegrees=45
  swing1Motion=Free
  swing2LimitDegrees=45
  swing2Motion=Free
  twistLimitDegrees=45
  twistMotion=Free
 }
 {
  bone1Name=cape
  bone2Name=spine_05
  swing1LimitDegrees=45
  swing1Motion=Free
  swing2LimitDegrees=45
  swing2Motion=Free
  twistLimitDegrees=45
  twistMotion=Free
 }
 {
  bone1Name=upperarm_l
  bone2Name=clavicle_l
  swing1LimitDegrees=45
  swing1Motion=Free
  swing2LimitDegrees=45
  swing2Motion=Free
  twistLimitDegrees=45
  twistMotion=Free
 }
 {
  bone1Name=upperarm_r
  bone2Name=clavicle_r
  swing1LimitDegrees=45
  swing1Motion=Free
  swing2LimitDegrees=45
  swing2Motion=Free
  twistLimitDegrees=45
  twistMotion=Free
 }
]
